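# Pre-sequence cinematic shot A0010 for WildEstate: one long camera move around the estate house that finally climbs high into the sky, where the Battle Bus appears and flies.
# Cam_A0010 uses a numbered rail, orbit, pan/boom and tilt/roll control rig with attach tracks, plus animated focal length and a template animation.
# Timed visibility toggles on walls and props, a swinging door pivot, and a moving TV and spotlight dress the move.
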LevelSequence /WildEstate/Sequences/Pre/A0010/WildEstate_Pre_A0010_02
bindings [
 BP_Blockout_Doorway_A_A_A75 {
  id=#1
  possessed_class=BP_WildEstate_Blockout_Doorway_Side_A_B_C
  tracks [{
    type=MovieSceneVisibilityTrack
    sections [{
      type=MovieSceneVisibilitySection
      channels [None {type=MovieSceneScriptingBoolChannel, default=false, keys [[1001,true],[1190,false],[1439,true]]}]
     }]
   }]
 }
 TestScaleEditorActor3 {
  id=#2
  possessed_class=TestScaleEditorActor_C
  tracks [{
    type=MovieSceneVisibilityTrack
    sections [{type=MovieSceneVisibilitySection, channels [None {type=MovieSceneScriptingBoolChannel, default=false}]}]
   }]
 }
 _A0010_WorldCube {
  id=#3
  possessed_class=FortStaticMeshActor
  tracks [
   {type=MovieSceneSpawnTrack, sections [{type=MovieSceneSpawnSection, channels [None {type=MovieSceneScriptingBoolChannel, default=true}]}]}
   {
    type=MovieScene3DTransformTrack
    sections [{
      type=MovieScene3DTransformSection
      channels [
       Location.X {type=MovieSceneScriptingDoubleChannel, default=7205.2965}
       Location.Y {type=MovieSceneScriptingDoubleChannel, default=11814.1699}
       Location.Z {type=MovieSceneScriptingDoubleChannel, default=366.018}
       Rotation.X {type=MovieSceneScriptingDoubleChannel, default=0}
       Rotation.Y {type=MovieSceneScriptingDoubleChannel, default=0}
       Rotation.Z {type=MovieSceneScriptingDoubleChannel, default=0}
       Scale.X {type=MovieSceneScriptingDoubleChannel, default=1}
       Scale.Y {type=MovieSceneScriptingDoubleChannel, default=1}
       Scale.Z {type=MovieSceneScriptingDoubleChannel, default=1}
      ]
     }]
   }
  ]
 }
 Cam_A0010 {
  id=#4
  possessed_class=CineCameraActor
  tracks [
   {
    type=MovieScene3DTransformTrack
    sections [{
      type=MovieScene3DTransformSection
      channels [
       Location.X {type=MovieSceneScriptingDoubleChannel, default=0, const [0,23]}
       Location.Y {type=MovieSceneScriptingDoubleChannel, default=0, const [0,23]}
       Location.Z {type=MovieSceneScriptingDoubleChannel, default=0, const [0,23]}
       Rotation.X {
        type=MovieSceneScriptingDoubleChannel
        default=0
        keys [[1109,0],[1121,0],[1134,-3.3656],[1157,-21.6205],[1165,-26.4799],[1177,-21.5168],[1195,-12.1924],[1226,-4.877],[1338,0],[1390,0],[1437,-2.4644],[1478,-6.7022],[1560,-17.945],[1662,-13.7951],[1749,-2.7304],[1792,-0.4914],[1826,0],[1964,0],[1998,-2.5824],[2070,-9.2856],[2159,15.1098],[2238,0],[2276,-4.8435],[2349,0],[2478,0],[2528,3.5936],[2625,0]]
       }
       Rotation.Y {type=MovieSceneScriptingDoubleChannel, default=0, thinned_from=23, keys [[1109,0],[2625,0]]}
       Rotation.Z {type=MovieSceneScriptingDoubleChannel, default=0, thinned_from=23, keys [[1109,0],[2625,0]]}
       Scale.X {type=MovieSceneScriptingDoubleChannel, default=1, const [1,23]}
       Scale.Y {type=MovieSceneScriptingDoubleChannel, default=1, const [1,23]}
       Scale.Z {type=MovieSceneScriptingDoubleChannel, default=1, const [1,23]}
      ]
     }]
   }
   {
    type=MovieSceneSpawnTrack
    sections [{type=MovieSceneSpawnSection, channels [None {type=MovieSceneScriptingBoolChannel, default=true, const [true,2]}]}]
   }
   {type=MovieScene3DAttachTrack, sections [{type=MovieScene3DAttachSection}]}
   {
    type=TemplateSequenceTrack
    display_name=Template Animation
    sections [{
      type=TemplateSequenceSection
      end_frame=4338
      start_frame=3167
      channels ["Transform[Rotation]" {type=MovieSceneScriptingFloatChannel, default=0, keys [[3167,0],[3194,18.956],[3238,0]]}]
     }]
   }
  ]
 }
 CineCameraActor {
  id=#5
  possessed_class=CineCameraActor
  tracks [
   {
    type=MovieScene3DTransformTrack
    sections [{
      type=MovieScene3DTransformSection
      channels [
       Location.X {type=MovieSceneScriptingDoubleChannel, default=7567.6929, const [7567.6929,1]}
       Location.Y {type=MovieSceneScriptingDoubleChannel, default=11083.6008, const [11083.6006,1]}
       Location.Z {type=MovieSceneScriptingDoubleChannel, default=621.1719, const [621.1719,1]}
       Rotation.X {type=MovieSceneScriptingDoubleChannel, default=0, const [0,1]}
       Rotation.Y {type=MovieSceneScriptingDoubleChannel, default=-9.4, const [-9.4,1]}
       Rotation.Z {type=MovieSceneScriptingDoubleChannel, default=114.4, const [114.4,1]}
       Scale.X {type=MovieSceneScriptingDoubleChannel, default=1, const [1,1]}
       Scale.Y {type=MovieSceneScriptingDoubleChannel, default=1, const [1,1]}
       Scale.Z {type=MovieSceneScriptingDoubleChannel, default=1, const [1,1]}
      ]
     }]
   }
   {type=MovieSceneSpawnTrack, sections [{type=MovieSceneSpawnSection, channels [None {type=MovieSceneScriptingBoolChannel, default=true}]}]}
  ]
 }
 GeoScript_WildEstate_VTXColor2 {
  id=#6
  possessed_class=GeoScript_WildEstate_VTXColor_C
  tracks [{
    type=MovieSceneVisibilityTrack
    sections [{type=MovieSceneVisibilitySection, channels [None {type=MovieSceneScriptingBoolChannel, default=false}]}]
   }]
 }
 BP_Blockout_Stairs_Rail_A_TESTTT {
  id=#7
  possessed_class=BP_Estate_Stairs_Rail_A_B_C
  tracks [{
    type=MovieSceneVisibilityTrack
    sections [{
      type=MovieSceneVisibilitySection
      channels [None {type=MovieSceneScriptingBoolChannel, default=false, keys [[1001,true],[1190,false],[1439,true]]}]
     }]
   }]
 }
 BP_Blockout_Doorway_A_A_A9 {
  id=#8
  possessed_class=BP_WildEstate_Blockout_Wall_A_A_C
  tracks [{
    type=MovieSceneVisibilityTrack
    sections [{type=MovieSceneVisibilitySection, channels [None {type=MovieSceneScriptingBoolChannel, default=true, keys [[1001,false],[1190,true]]}]}]
   }]
 }
 CameraComponent {
  id=#9
  parent_id=#5
  possessed_class=CineCameraComponent
  tracks [
   {
    type=MovieSceneFloatTrack
    display_name=Current Focal Length
    sections [{
      type=MovieSceneFloatSection
      channels [None {type=MovieSceneScriptingFloatChannel, default=14.59, thinned_from=5, keys [[1151,38],[1191,14],[1993,14],[2042,14.59]]}]
     }]
   }
   {
    type=MovieSceneFloatTrack
    display_name="Manual Focus Distance (Focus Settings)"
    sections [{type=MovieSceneFloatSection, channels [None {type=MovieSceneScriptingFloatChannel, default=100000}]}]
   }
   {
    type=MovieSceneFloatTrack
    display_name=Current Aperture
    sections [{type=MovieSceneFloatSection, channels [None {type=MovieSceneScriptingFloatChannel, default=2.8}]}]
   }
   {
    type=MovieScene3DTransformTrack
    sections [{
      type=MovieScene3DTransformSection
      channels [
       Location.X {type=MovieSceneScriptingDoubleChannel, default=0, const [0,1]}
       Location.Y {type=MovieSceneScriptingDoubleChannel, default=58.2237, const [0,1]}
       Location.Z {type=MovieSceneScriptingDoubleChannel, default=0, const [0,1]}
       Rotation.X {type=MovieSceneScriptingDoubleChannel, default=0, const [0,1]}
       Rotation.Y {type=MovieSceneScriptingDoubleChannel, default=0, const [0,1]}
       Rotation.Z {type=MovieSceneScriptingDoubleChannel, default=0, const [0,1]}
       Scale.X {type=MovieSceneScriptingDoubleChannel, default=1, const [1,1]}
       Scale.Y {type=MovieSceneScriptingDoubleChannel, default=1, const [1,1]}
       Scale.Z {type=MovieSceneScriptingDoubleChannel, default=1, const [1,1]}
      ]
     }]
   }
  ]
 }
 SM_WildEstate_House_Door_A {
  id=#10
  possessed_class=FortStaticMeshActor
  tracks [
   {type=MovieSceneSpawnTrack, sections [{type=MovieSceneSpawnSection, channels [None {type=MovieSceneScriptingBoolChannel, default=true}]}]}
   {
    type=MovieScene3DTransformTrack
    sections [{
      type=MovieScene3DTransformSection
      channels [
       Location.X {type=MovieSceneScriptingDoubleChannel, default=-85.583}
       Location.Y {type=MovieSceneScriptingDoubleChannel, default=0}
       Location.Z {type=MovieSceneScriptingDoubleChannel, default=0}
       Rotation.X {type=MovieSceneScriptingDoubleChannel, default=0}
       Rotation.Y {type=MovieSceneScriptingDoubleChannel, default=0}
       Rotation.Z {type=MovieSceneScriptingDoubleChannel, default=0}
       Scale.X {type=MovieSceneScriptingDoubleChannel, default=1}
       Scale.Y {type=MovieSceneScriptingDoubleChannel, default=1}
       Scale.Z {type=MovieSceneScriptingDoubleChannel, default=1}
      ]
     }]
   }
   {type=MovieScene3DAttachTrack, sections [{type=MovieScene3DAttachSection}]}
  ]
 }
 BP_Blockout_Doorway_A_A_A73 {
  id=#11
  possessed_class=BP_WildEstate_Blockout_Doorway_Side_A_B_C
  tracks [{
    type=MovieSceneVisibilityTrack
    sections [{
      type=MovieSceneVisibilitySection
      channels [None {type=MovieSceneScriptingBoolChannel, default=false, keys [[1001,true],[1190,false],[1439,true]]}]
     }]
   }]
 }
 LightComponent0 {
  id=#12
  parent_id=#13
  possessed_class=SpotLightComponent
  tracks [
   {
    type=MovieSceneFloatTrack
    display_name=Intensity
    sections [{type=MovieSceneFloatSection, channels [None {type=MovieSceneScriptingFloatChannel, default=4}]}]
   }
   {
    type=MovieSceneColorTrack
    display_name=Light Color
    sections [{
      type=MovieSceneColorSection
      channels [
       Color.R {type=MovieSceneScriptingFloatChannel, default=1}
       Color.G {type=MovieSceneScriptingFloatChannel, default=1}
       Color.B {type=MovieSceneScriptingFloatChannel, default=1}
       Color.A {type=MovieSceneScriptingFloatChannel, default=1}
      ]
     }]
   }
  ]
 }
 BP_WildEstate_House_TV_A {
  id=#14
  possessed_class=BP_WildEstate_House_TV_A_C
  tracks [{
    type=MovieScene3DTransformTrack
    sections [{
      type=MovieScene3DTransformSection
      channels [
       Location.X {type=MovieSceneScriptingDoubleChannel, default=7265.8323, keys [[1001,7200.8853],[1439,7265.832]]}
       Location.Y {type=MovieSceneScriptingDoubleChannel, default=11572.4742, keys [[1001,11571.4629],[1439,11560.4746]]}
       Location.Z {type=MovieSceneScriptingDoubleChannel, default=394.9222, keys [[1001,371.7595],[1439,394.9221]]}
       Rotation.X {type=MovieSceneScriptingDoubleChannel, default=0, const [0,2]}
       Rotation.Y {type=MovieSceneScriptingDoubleChannel, default=0, const [0,2]}
       Rotation.Z {type=MovieSceneScriptingDoubleChannel, default=0, const [0,2]}
       Scale.X {type=MovieSceneScriptingDoubleChannel, default=1, const [1,2]}
       Scale.Y {type=MovieSceneScriptingDoubleChannel, default=1, const [1,2]}
       Scale.Z {type=MovieSceneScriptingDoubleChannel, default=1, const [1,2]}
      ]
     }]
   }]
 }
 TestScaleEditorActor2 {
  id=#15
  possessed_class=TestScaleEditorActor_C
  tracks [{
    type=MovieSceneVisibilityTrack
    sections [{type=MovieSceneVisibilitySection, channels [None {type=MovieSceneScriptingBoolChannel, default=false}]}]
   }]
 }
 DoorComponent {
  id=#16
  parent_id=#17
  possessed_class=FortBuildingWallDoorComponent
  tracks [
   {
    type=MovieScene3DTransformTrack
    sections [{
      type=MovieScene3DTransformSection
      channels [
       Location.X {type=MovieSceneScriptingDoubleChannel, default=0}
       Location.Y {type=MovieSceneScriptingDoubleChannel, default=-1}
       Location.Z {type=MovieSceneScriptingDoubleChannel, default=1}
       Rotation.X {type=MovieSceneScriptingDoubleChannel, default=0}
       Rotation.Y {type=MovieSceneScriptingDoubleChannel, default=0}
       Rotation.Z {type=MovieSceneScriptingDoubleChannel, default=0}
       Scale.X {type=MovieSceneScriptingDoubleChannel, default=1}
       Scale.Y {type=MovieSceneScriptingDoubleChannel, default=1}
       Scale.Z {type=MovieSceneScriptingDoubleChannel, default=1}
      ]
     }]
   }
   {
    type=MovieSceneVisibilityTrack
    sections [{type=MovieSceneVisibilitySection, channels [None {type=MovieSceneScriptingBoolChannel, default=false}]}]
   }
  ]
 }
 BrickSimple_Chimney_B {
  id=#18
  possessed_class=BP_WildEstate_House_Fireplace_A_C
  tracks [{
    type=MovieSceneVisibilityTrack
    sections [{type=MovieSceneVisibilitySection, channels [None {type=MovieSceneScriptingBoolChannel, default=true, keys [[1001,false],[1190,true]]}]}]
   }]
 }
 HoodLOD0_LOD0 {
  id=#19
  possessed_class=FortStaticMeshActor
  tracks [
   {
    type=MovieScene3DTransformTrack
    sections [{
      type=MovieScene3DTransformSection
      channels [
       Location.X {type=MovieSceneScriptingDoubleChannel, default=16952.6667}
       Location.Y {type=MovieSceneScriptingDoubleChannel, default=3160.7244}
       Location.Z {type=MovieSceneScriptingDoubleChannel, default=10.8123}
       Rotation.X {type=MovieSceneScriptingDoubleChannel, default=0}
       Rotation.Y {type=MovieSceneScriptingDoubleChannel, default=0}
       Rotation.Z {type=MovieSceneScriptingDoubleChannel, default=157.4188}
       Scale.X {type=MovieSceneScriptingDoubleChannel, default=1}
       Scale.Y {type=MovieSceneScriptingDoubleChannel, default=1}
       Scale.Z {type=MovieSceneScriptingDoubleChannel, default=1}
      ]
     }]
   }
   {
    type=MovieSceneVisibilityTrack
    sections [{type=MovieSceneVisibilitySection, channels [None {type=MovieSceneScriptingBoolChannel, default=false}]}]
   }
  ]
 }
 _1_BaseOrbit_CTRL {
  id=#20
  possessed_class=FortStaticMeshActor
  tracks [
   {type=MovieSceneSpawnTrack, sections [{type=MovieSceneSpawnSection, channels [None {type=MovieSceneScriptingBoolChannel, default=true}]}]}
   {
    type=MovieScene3DTransformTrack
    sections [{
      type=MovieScene3DTransformSection
      channels [
       Location.X {type=MovieSceneScriptingDoubleChannel, default=18.5133, const [0,3]}
       Location.Y {type=MovieSceneScriptingDoubleChannel, default=0, const [0,3]}
       Location.Z {type=MovieSceneScriptingDoubleChannel, default=0, const [0,3]}
       Rotation.X {type=MovieSceneScriptingDoubleChannel, default=0, const [0,3]}
       Rotation.Y {type=MovieSceneScriptingDoubleChannel, default=0, const [0,3]}
       Rotation.Z {
        type=MovieSceneScriptingDoubleChannel
        default=114.737
        keys [[1001,28.185],[1121,28.185],[1133,26.0622],[1167,4.3495],[1189,-33.0399],[1211,-61.3435],[1251,-82.4931],[1338,-113.158],[1437,-143.4369],[1486,-197.6509],[1499,-230.4551],[1515,-250.9328],[1554,-264.8864],[1575,-281.8762],[1605,-343.5479],[1639,-401.4576],[1676,-409.7337],[1826,-413.8355],[1964,-413.9029],[1998,-432.3197],[2070,-490.365],[2125,-454.674],[2151,-447.7767],[2224,-461.2338],[2280,-483.2837],[2335,-491.9114],[2378,-486.5244],[2402,-478.1369],[2482,-439.8911],[2544,-387.4397],[2582,-382.7825],[2657,-393.8647],[2746,-426.8457],[2805,-459.4213],[2845,-504.2332],[3038,-759.2437],[3159,-765.5668]]
       }
       Scale.X {type=MovieSceneScriptingDoubleChannel, default=1, const [1,3]}
       Scale.Y {type=MovieSceneScriptingDoubleChannel, default=1, const [1,3]}
       Scale.Z {type=MovieSceneScriptingDoubleChannel, default=1, const [1,3]}
      ]
     }]
   }
   {type=MovieScene3DAttachTrack, sections [{type=MovieScene3DAttachSection}]}
  ]
 }
 BP_Blockout_Doorway_A_A_A8 {
  id=#21
  possessed_class=BP_WildEstate_House_Window_C_A_C
  tracks [{
    type=MovieSceneVisibilityTrack
    sections [{type=MovieSceneVisibilitySection, channels [None {type=MovieSceneScriptingBoolChannel, default=true, keys [[1001,false],[1190,true]]}]}]
   }]
 }
 SpotLight {
  id=#13
  possessed_class=SpotLight
  tracks [{
    type=MovieScene3DTransformTrack
    sections [{
      type=MovieScene3DTransformSection
      channels [
       Location.X {type=MovieSceneScriptingDoubleChannel, default=7264, keys [[1001,7240.2725],[1439,7264]]}
       Location.Y {type=MovieSceneScriptingDoubleChannel, default=11552, keys [[1001,11608.0605],[1439,11572]]}
       Location.Z {type=MovieSceneScriptingDoubleChannel, default=466, keys [[1001,440.0335],[1439,466]]}
       Rotation.X {type=MovieSceneScriptingDoubleChannel, default=-90, const [-90,2]}
       Rotation.Y {type=MovieSceneScriptingDoubleChannel, default=-2.8125, const [-2.8125,2]}
       Rotation.Z {type=MovieSceneScriptingDoubleChannel, default=90, const [90,2]}
       Scale.X {type=MovieSceneScriptingDoubleChannel, default=1, const [1,2]}
       Scale.Y {type=MovieSceneScriptingDoubleChannel, default=1, const [1,2]}
       Scale.Z {type=MovieSceneScriptingDoubleChannel, default=1, const [1,2]}
      ]
     }]
   }]
 }
 _3_TiltRoll_CTRL {
  id=#22
  possessed_class=FortStaticMeshActor
  tracks [
   {type=MovieSceneSpawnTrack, sections [{type=MovieSceneSpawnSection, channels [None {type=MovieSceneScriptingBoolChannel, default=true}]}]}
   {
    type=MovieScene3DTransformTrack
    sections [{
      type=MovieScene3DTransformSection
      channels [
       Location.X {type=MovieSceneScriptingDoubleChannel, default=0, thinned_from=10, keys [[1001,-55.3521],[1121,-0.8799],[1133,-0.0058],[3308,-0.0058]]}
       Location.Y {type=MovieSceneScriptingDoubleChannel, default=0, thinned_from=10, keys [[1001,0],[3308,0]]}
       Location.Z {type=MovieSceneScriptingDoubleChannel, default=0, thinned_from=10, keys [[1001,9.4262],[1121,0.1498],[1133,0.001],[3308,0.001]]}
       Rotation.X {type=MovieSceneScriptingDoubleChannel, default=0, const [0,8]}
       Rotation.Y {
        type=MovieSceneScriptingDoubleChannel
        default=-7.6791
        thinned_from=28
        keys [[1001,-9.6645],[1121,-9.6645],[1133,-8.9973],[1338,7.0985],[1562,7.0985],[1672,0.0545],[1775,-5.648],[1826,-6.2743],[2042,-6.4574],[2225,-0.7815],[2278,10.6631],[2359,21.1804],[2418,18.6941],[2482,9.0227],[2566,6.6498],[2657,5.857],[2716,0.9761],[2805,-19.4865],[3038,-32.2329],[3180,-32.2329],[3235,-8.8027],[3284,1.2356],[3306,2.3135],[3380,3.3695],[3618,3.4658],[3741,36.0796]]
       }
       Rotation.Z {type=MovieSceneScriptingDoubleChannel, default=0, const [0,8]}
       Scale.X {type=MovieSceneScriptingDoubleChannel, default=1, const [1,8]}
       Scale.Y {type=MovieSceneScriptingDoubleChannel, default=1, const [1,8]}
       Scale.Z {type=MovieSceneScriptingDoubleChannel, default=1, const [1,8]}
      ]
     }]
   }
   {type=MovieScene3DAttachTrack, sections [{type=MovieScene3DAttachSection}]}
  ]
 }
 CameraComponent {
  id=#23
  parent_id=#4
  possessed_class=CineCameraComponent
  tracks [
   {
    type=MovieSceneFloatTrack
    display_name=Current Focal Length
    sections [{
      type=MovieSceneFloatSection
      channels [None {
        type=MovieSceneScriptingFloatChannel
        default=30
        thinned_from=13
        keys [[1121,30],[1169,26.175],[1203,19.4077],[1249,17],[1478,17],[1545,17.5408],[1597,19.8146],[1659,25.0863],[1727,28],[1964,28],[1985,15.9571],[1998,14]]
       }]
     }]
   }
   {
    type=MovieSceneFloatTrack
    display_name="Manual Focus Distance (Focus Settings)"
    sections [{type=MovieSceneFloatSection, channels [None {type=MovieSceneScriptingFloatChannel, default=0, const [0,3]}]}]
   }
   {
    type=MovieSceneFloatTrack
    display_name=Current Aperture
    sections [{type=MovieSceneFloatSection, channels [None {type=MovieSceneScriptingFloatChannel, default=2.8, const [2.8,2]}]}]
   }
   {
    type=MovieScene3DTransformTrack
    sections [{
      type=MovieScene3DTransformSection
      channels [
       Location.X {type=MovieSceneScriptingDoubleChannel, default=0, const [0,4]}
       Location.Y {type=MovieSceneScriptingDoubleChannel, default=0, const [0,4]}
       Location.Z {type=MovieSceneScriptingDoubleChannel, default=0, const [0,4]}
       Rotation.X {type=MovieSceneScriptingDoubleChannel, default=0, const [0,4]}
       Rotation.Y {type=MovieSceneScriptingDoubleChannel, default=0, const [0,4]}
       Rotation.Z {type=MovieSceneScriptingDoubleChannel, default=0, const [0,4]}
       Scale.X {type=MovieSceneScriptingDoubleChannel, default=1, const [1,4]}
       Scale.Y {type=MovieSceneScriptingDoubleChannel, default=1, const [1,4]}
       Scale.Z {type=MovieSceneScriptingDoubleChannel, default=1, const [1,4]}
      ]
     }]
   }
  ]
 }
 _2_PanBoom_CTRL {
  id=#24
  possessed_class=FortStaticMeshActor
  tracks [
   {type=MovieSceneSpawnTrack, sections [{type=MovieSceneSpawnSection, channels [None {type=MovieSceneScriptingBoolChannel, default=true}]}]}
   {
    type=MovieScene3DTransformTrack
    sections [{
      type=MovieScene3DTransformSection
      channels [
       Location.X {
        type=MovieSceneScriptingDoubleChannel
        default=-806.4399
        thinned_from=18
        keys [[1121,94.905],[1133,80.4092],[1189,2.7086],[1338,-27.8719],[1438,-42.7633],[1459,-43.4963],[1550,-21.1465],[1620,-5.6911],[1950,0],[2042,-40.4706],[2278,236.4027],[2482,147.8211],[2657,-385.9533],[2746,-760.9737],[2805,-884.2246],[2845,-887.9854],[3038,-46151.1445]]
       }
       Location.Y {
        type=MovieSceneScriptingDoubleChannel
        default=-7.2342
        thinned_from=18
        keys [[1121,-123.1737],[1133,-113.6576],[1168,-9.1734],[1338,0],[3038,0]]
       }
       Location.Z {
        type=MovieSceneScriptingDoubleChannel
        default=252.6591
        keys [[1121,204.4086],[1133,194.686],[1211,118.3018],[1338,62.1198],[1438,21.807],[1496,11.3668],[1559,32.6016],[1603,76.7968],[1672,426.6545],[1692,560.3425],[1826,965.1935],[1964,987.6728],[2042,144.5846],[2225,27.4371],[2278,20.4652],[2418,32.5326],[2482,60.6879],[2524,94.2107],[2566,165.1166],[2657,425.7182],[2746,1010.1359],[2805,1424.6883],[2845,1805.067],[3038,38208.9062]]
       }
       Rotation.X {type=MovieSceneScriptingDoubleChannel, default=0, const [0,4]}
       Rotation.Y {type=MovieSceneScriptingDoubleChannel, default=0, const [0,4]}
       Rotation.Z {
        type=MovieSceneScriptingDoubleChannel
        default=-0.582
        thinned_from=19
        keys [[1121,0.5785],[1133,1.1817],[1168,4.5236],[1189,8.7477],[1211,8.6368],[1338,0],[3038,0]]
       }
       Scale.X {type=MovieSceneScriptingDoubleChannel, default=1, const [1,4]}
       Scale.Y {type=MovieSceneScriptingDoubleChannel, default=1, const [1,4]}
       Scale.Z {type=MovieSceneScriptingDoubleChannel, default=1, const [1,4]}
      ]
     }]
   }
   {type=MovieScene3DAttachTrack, sections [{type=MovieScene3DAttachSection}]}
  ]
 }
 _0_CameraRig_Rail {
  id=#25
  possessed_class=CameraRig_Rail
  tracks [
   {
    type=MovieSceneSpawnTrack
    sections [{type=MovieSceneSpawnSection, channels [None {type=MovieSceneScriptingBoolChannel, default=true, const [true,2]}]}]
   }
   {
    type=MovieScene3DTransformTrack
    sections [{
      type=MovieScene3DTransformSection
      channels [
       Location.X {type=MovieSceneScriptingDoubleChannel, default=7464.3386, thinned_from=5, keys [[1001,7464.3386],[2241,7464.3384]]}
       Location.Y {type=MovieSceneScriptingDoubleChannel, default=11012.2387, thinned_from=5, keys [[1001,11012.2387],[2241,11012.2383]]}
       Location.Z {type=MovieSceneScriptingDoubleChannel, default=422.674, thinned_from=5, keys [[1003,422.674],[2241,422.674]]}
       Rotation.X {type=MovieSceneScriptingDoubleChannel, default=0, const [0,5]}
       Rotation.Y {type=MovieSceneScriptingDoubleChannel, default=0, const [0,5]}
       Rotation.Z {type=MovieSceneScriptingDoubleChannel, default=86.8822, thinned_from=5, keys [[1003,86.8822],[2241,86.8822]]}
       Scale.X {type=MovieSceneScriptingDoubleChannel, default=1, const [1,5]}
       Scale.Y {type=MovieSceneScriptingDoubleChannel, default=1, const [1,5]}
       Scale.Z {type=MovieSceneScriptingDoubleChannel, default=1, const [1,5]}
      ]
     }]
   }
   {
    type=MovieSceneFloatTrack
    display_name=Current Position on Rail
    sections [{
      type=MovieSceneFloatSection
      channels [None {
        type=MovieSceneScriptingFloatChannel
        default=0.1026
        keys [[1001,0.0024],[1121,0.0024],[1130,0.0065],[1172,0.0308],[1252,0.0473],[1435,0.0625],[1495,0.0806],[1755,0.2337],[1826,0.248],[1964,0.25],[1979,0.2813],[2229,0.6522],[2466,0.7968],[2636,0.8785],[3009,0.9816]]
       }]
     }]
   }
  ]
 }
 BP_Blockout_Doorway_A_A_A17 {
  id=#26
  possessed_class=BP_WildEstate_Blockout_Wall_A_A_C
  tracks [{
    type=MovieSceneVisibilityTrack
    sections [{
      type=MovieSceneVisibilitySection
      channels [None {type=MovieSceneScriptingBoolChannel, default=false, keys [[1001,true],[1190,false],[1439,true]]}]
     }]
   }]
 }
 BP_Blockout_Doorway_A_A_A {id=#17, possessed_class=BP_WildEstate_House_Door_A_A_C, tracks []}
 Artemis_Suburban_CurtainsModern_C3 {
  id=#27
  possessed_class=Artemis_Suburban_CurtainsModern_C_C
  tracks [{
    type=MovieSceneVisibilityTrack
    sections [{type=MovieSceneVisibilitySection, channels [None {type=MovieSceneScriptingBoolChannel, default=true, keys [[1001,false],[1190,true]]}]}]
   }]
 }
 door_pivot {
  id=#28
  possessed_class=FortStaticMeshActor
  tracks [
   {type=MovieSceneSpawnTrack, sections [{type=MovieSceneSpawnSection, channels [None {type=MovieSceneScriptingBoolChannel, default=true}]}]}
   {
    type=MovieScene3DTransformTrack
    sections [{
      type=MovieScene3DTransformSection
      channels [
       Location.X {type=MovieSceneScriptingDoubleChannel, default=6400.9955, const [6400.9956,6]}
       Location.Y {type=MovieSceneScriptingDoubleChannel, default=12288, const [12373.583,6]}
       Location.Z {type=MovieSceneScriptingDoubleChannel, default=384.8292, const [384.8292,6]}
       Rotation.X {type=MovieSceneScriptingDoubleChannel, default=0, const [0,6]}
       Rotation.Y {type=MovieSceneScriptingDoubleChannel, default=0, const [0,6]}
       Rotation.Z {type=MovieSceneScriptingDoubleChannel, default=90, keys [[1518,90],[1524,142.2273],[1536,183.515],[1584,183.515],[1591,155.355],[1604,90]]}
       Scale.X {type=MovieSceneScriptingDoubleChannel, default=1, const [1,6]}
       Scale.Y {type=MovieSceneScriptingDoubleChannel, default=1, const [1,6]}
       Scale.Z {type=MovieSceneScriptingDoubleChannel, default=1, const [1,6]}
      ]
     }]
   }
  ]
 }
 B_BattleBus_Stamina {
  id=#29
  possessed_class=B_BattleBus_Stamina_C
  tracks [
   {type=MovieSceneSpawnTrack, sections [{type=MovieSceneSpawnSection, channels [None {type=MovieSceneScriptingBoolChannel, default=true}]}]}
   {
    type=MovieScene3DTransformTrack
    sections [{
      type=MovieScene3DTransformSection
      channels [
       Location.X {type=MovieSceneScriptingDoubleChannel, default=-19250.0933, keys [[3180,-19250.0938],[3279,-16774.5684]]}
       Location.Y {type=MovieSceneScriptingDoubleChannel, default=-25564.6301, keys [[3180,-25564.6309],[3279,-23150.627]]}
       Location.Z {type=MovieSceneScriptingDoubleChannel, default=38281.5445, const [38281.543,2]}
       Rotation.X {type=MovieSceneScriptingDoubleChannel, default=0, const [0,2]}
       Rotation.Y {type=MovieSceneScriptingDoubleChannel, default=0, const [0,2]}
       Rotation.Z {type=MovieSceneScriptingDoubleChannel, default=44.2791, const [44.2791,2]}
       Scale.X {type=MovieSceneScriptingDoubleChannel, default=1, const [1,2]}
       Scale.Y {type=MovieSceneScriptingDoubleChannel, default=1, const [1,2]}
       Scale.Z {type=MovieSceneScriptingDoubleChannel, default=1, const [1,2]}
      ]
     }]
   }
   {
    type=MovieSceneVisibilityTrack
    sections [{type=MovieSceneVisibilitySection, channels [None {type=MovieSceneScriptingBoolChannel, default=true, keys [[1001,false],[3128,true]]}]}]
   }
  ]
 }
 SM_WildEstate_House_PictureFrames_A {
  id=#30
  possessed_class=BP_WildEstate_House_PictureFrames_A_C
  tracks [{
    type=MovieSceneVisibilityTrack
    sections [{type=MovieSceneVisibilitySection, channels [None {type=MovieSceneScriptingBoolChannel, default=true, keys [[1001,false],[1190,true]]}]}]
   }]
 }
]
display_rate {denominator=1, numerator=60}
tick_resolution {denominator=1, numerator=24000}
tracks [{type=MovieSceneCameraCutTrack, sections [{type=MovieSceneCameraCutSection, end_frame=3566, start_frame=737}]}]
