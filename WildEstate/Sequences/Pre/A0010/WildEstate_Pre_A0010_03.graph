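# Cinematic Level Sequence for WildEstate Pre shot A0010_03: an animated camera move around the estate house, with one camera-cut section.
# Cam_A0010 combines a rail with attached orbit, pan/boom and tilt/roll controls plus animated zoom; a second, static camera also zooms.
# Timed visibility swaps hide and reveal walls, doorways and props mid-shot.
# A spawned pivot swings a door open and shut, and the Battle Bus becomes visible late on.
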LevelSequence /WildEstate/Sequences/Pre/A0010/WildEstate_Pre_A0010_03
bindings [
 BP_Blockout_Doorway_A_A_A75 {
  id=#1
  possessed_class=BP_WildEstate_Blockout_Doorway_Side_A_B_C
  tracks [{
    type=MovieSceneVisibilityTrack
    sections [{
      type=MovieSceneVisibilitySection
      channels [None {type=MovieSceneScriptingBoolChannel, default=false, keys [[1001,true],[1180,false],[1379,true]]}]
     }]
   }]
 }
 TestScaleEditorActor3 {
  id=#2
  possessed_class=TestScaleEditorActor_C
  tracks [{
    type=MovieSceneVisibilityTrack
    sections [{type=MovieSceneVisibilitySection, channels [None {type=MovieSceneScriptingBoolChannel, default=false}]}]
   }]
 }
 _A0010_WorldCube {
  id=#3
  possessed_class=FortStaticMeshActor
  tracks [
   {type=MovieSceneSpawnTrack, sections [{type=MovieSceneSpawnSection, channels [None {type=MovieSceneScriptingBoolChannel, default=true}]}]}
   {
    type=MovieScene3DTransformTrack
    sections [{
      type=MovieScene3DTransformSection
      channels [
       Location.X {type=MovieSceneScriptingDoubleChannel, default=7205.2965}
       Location.Y {type=MovieSceneScriptingDoubleChannel, default=11814.1699}
       Location.Z {type=MovieSceneScriptingDoubleChannel, default=366.018}
       Rotation.X {type=MovieSceneScriptingDoubleChannel, default=0}
       Rotation.Y {type=MovieSceneScriptingDoubleChannel, default=0}
       Rotation.Z {type=MovieSceneScriptingDoubleChannel, default=0}
       Scale.X {type=MovieSceneScriptingDoubleChannel, default=1}
       Scale.Y {type=MovieSceneScriptingDoubleChannel, default=1}
       Scale.Z {type=MovieSceneScriptingDoubleChannel, default=1}
      ]
     }]
   }
  ]
 }
 Cam_A0010 {
  id=#4
  possessed_class=CineCameraActor
  tracks [
   {
    type=MovieScene3DTransformTrack
    sections [{
      type=MovieScene3DTransformSection
      channels [
       Location.X {type=MovieSceneScriptingDoubleChannel, default=0, const [0,16]}
       Location.Y {type=MovieSceneScriptingDoubleChannel, default=0, const [0,16]}
       Location.Z {type=MovieSceneScriptingDoubleChannel, default=0, const [0,16]}
       Rotation.X {
        type=MovieSceneScriptingDoubleChannel
        default=0
        thinned_from=11
        keys [[1121,0],[1134,-3.3656],[1157,-21.6205],[1165,-26.4799],[1177,-21.5168],[1195,-12.1924],[1226,-4.877],[1278,0],[1724,0]]
       }
       Rotation.Y {type=MovieSceneScriptingDoubleChannel, default=0, thinned_from=16, keys [[1121,0],[1724,0]]}
       Rotation.Z {type=MovieSceneScriptingDoubleChannel, default=0, thinned_from=16, keys [[1121,0],[1724,0]]}
       Scale.X {type=MovieSceneScriptingDoubleChannel, default=1, const [1,16]}
       Scale.Y {type=MovieSceneScriptingDoubleChannel, default=1, const [1,16]}
       Scale.Z {type=MovieSceneScriptingDoubleChannel, default=1, const [1,16]}
      ]
     }]
   }
   {
    type=MovieSceneSpawnTrack
    sections [{type=MovieSceneSpawnSection, channels [None {type=MovieSceneScriptingBoolChannel, default=true, const [true,2]}]}]
   }
   {type=MovieScene3DAttachTrack, sections [{type=MovieScene3DAttachSection}]}
  ]
 }
 CineCameraActor {
  id=#5
  possessed_class=CineCameraActor
  tracks [
   {
    type=MovieScene3DTransformTrack
    sections [{
      type=MovieScene3DTransformSection
      channels [
       Location.X {type=MovieSceneScriptingDoubleChannel, default=7567.6929, const [7567.6929,1]}
       Location.Y {type=MovieSceneScriptingDoubleChannel, default=11083.6008, const [11083.6006,1]}
       Location.Z {type=MovieSceneScriptingDoubleChannel, default=621.1719, const [621.1719,1]}
       Rotation.X {type=MovieSceneScriptingDoubleChannel, default=0, const [0,1]}
       Rotation.Y {type=MovieSceneScriptingDoubleChannel, default=-9.4, const [-9.4,1]}
       Rotation.Z {type=MovieSceneScriptingDoubleChannel, default=114.4, const [114.4,1]}
       Scale.X {type=MovieSceneScriptingDoubleChannel, default=1, const [1,1]}
       Scale.Y {type=MovieSceneScriptingDoubleChannel, default=1, const [1,1]}
       Scale.Z {type=MovieSceneScriptingDoubleChannel, default=1, const [1,1]}
      ]
     }]
   }
   {type=MovieSceneSpawnTrack, sections [{type=MovieSceneSpawnSection, channels [None {type=MovieSceneScriptingBoolChannel, default=true}]}]}
  ]
 }
 BP_Blockout_Stairs_Rail_A_TESTTT {
  id=#6
  possessed_class=BP_Estate_Stairs_Rail_A_B_C
  tracks [{
    type=MovieSceneVisibilityTrack
    sections [{
      type=MovieSceneVisibilitySection
      channels [None {type=MovieSceneScriptingBoolChannel, default=false, keys [[1001,true],[1180,false],[1379,true]]}]
     }]
   }]
 }
 BP_Blockout_Doorway_A_A_A9 {
  id=#7
  possessed_class=BP_WildEstate_Blockout_Wall_A_A_C
  tracks [{
    type=MovieSceneVisibilityTrack
    sections [{type=MovieSceneVisibilitySection, channels [None {type=MovieSceneScriptingBoolChannel, default=true, keys [[1001,false],[1180,true]]}]}]
   }]
 }
 CameraComponent {
  id=#8
  parent_id=#5
  possessed_class=CineCameraComponent
  tracks [
   {
    type=MovieSceneFloatTrack
    display_name=Current Focal Length
    sections [{
      type=MovieSceneFloatSection
      channels [None {type=MovieSceneScriptingFloatChannel, default=28, thinned_from=5, keys [[1151,38],[1191,14],[1753,14],[2149,28]]}]
     }]
   }
   {
    type=MovieSceneFloatTrack
    display_name="Manual Focus Distance (Focus Settings)"
    sections [{type=MovieSceneFloatSection, channels [None {type=MovieSceneScriptingFloatChannel, default=100000}]}]
   }
   {
    type=MovieSceneFloatTrack
    display_name=Current Aperture
    sections [{type=MovieSceneFloatSection, channels [None {type=MovieSceneScriptingFloatChannel, default=2.8}]}]
   }
   {
    type=MovieScene3DTransformTrack
    sections [{
      type=MovieScene3DTransformSection
      channels [
       Location.X {type=MovieSceneScriptingDoubleChannel, default=0, const [0,1]}
       Location.Y {type=MovieSceneScriptingDoubleChannel, default=58.2237, const [0,1]}
       Location.Z {type=MovieSceneScriptingDoubleChannel, default=0, const [0,1]}
       Rotation.X {type=MovieSceneScriptingDoubleChannel, default=0, const [0,1]}
       Rotation.Y {type=MovieSceneScriptingDoubleChannel, default=0, const [0,1]}
       Rotation.Z {type=MovieSceneScriptingDoubleChannel, default=0, const [0,1]}
       Scale.X {type=MovieSceneScriptingDoubleChannel, default=1, const [1,1]}
       Scale.Y {type=MovieSceneScriptingDoubleChannel, default=1, const [1,1]}
       Scale.Z {type=MovieSceneScriptingDoubleChannel, default=1, const [1,1]}
      ]
     }]
   }
  ]
 }
 SM_WildEstate_House_Door_A {
  id=#9
  possessed_class=FortStaticMeshActor
  tracks [
   {type=MovieSceneSpawnTrack, sections [{type=MovieSceneSpawnSection, channels [None {type=MovieSceneScriptingBoolChannel, default=true}]}]}
   {
    type=MovieScene3DTransformTrack
    sections [{
      type=MovieScene3DTransformSection
      channels [
       Location.X {type=MovieSceneScriptingDoubleChannel, default=-85.583}
       Location.Y {type=MovieSceneScriptingDoubleChannel, default=0}
       Location.Z {type=MovieSceneScriptingDoubleChannel, default=0}
       Rotation.X {type=MovieSceneScriptingDoubleChannel, default=0}
       Rotation.Y {type=MovieSceneScriptingDoubleChannel, default=0}
       Rotation.Z {type=MovieSceneScriptingDoubleChannel, default=0}
       Scale.X {type=MovieSceneScriptingDoubleChannel, default=1}
       Scale.Y {type=MovieSceneScriptingDoubleChannel, default=1}
       Scale.Z {type=MovieSceneScriptingDoubleChannel, default=1}
      ]
     }]
   }
   {type=MovieScene3DAttachTrack, sections [{type=MovieScene3DAttachSection}]}
  ]
 }
 BP_Blockout_Doorway_A_A_A73 {
  id=#10
  possessed_class=BP_WildEstate_Blockout_Doorway_Side_A_B_C
  tracks [{
    type=MovieSceneVisibilityTrack
    sections [{
      type=MovieSceneVisibilitySection
      channels [None {type=MovieSceneScriptingBoolChannel, default=false, keys [[1001,true],[1180,false],[1379,true]]}]
     }]
   }]
 }
 LightComponent0 {
  id=#11
  parent_id=#12
  possessed_class=SpotLightComponent
  tracks [
   {
    type=MovieSceneFloatTrack
    display_name=Intensity
    sections [{type=MovieSceneFloatSection, channels [None {type=MovieSceneScriptingFloatChannel, default=4}]}]
   }
   {
    type=MovieSceneColorTrack
    display_name=Light Color
    sections [{
      type=MovieSceneColorSection
      channels [
       Color.R {type=MovieSceneScriptingFloatChannel, default=1}
       Color.G {type=MovieSceneScriptingFloatChannel, default=1}
       Color.B {type=MovieSceneScriptingFloatChannel, default=1}
       Color.A {type=MovieSceneScriptingFloatChannel, default=1}
      ]
     }]
   }
  ]
 }
 BP_WildEstate_House_TV_A {
  id=#13
  possessed_class=BP_WildEstate_House_TV_A_C
  tracks [{
    type=MovieScene3DTransformTrack
    sections [{
      type=MovieScene3DTransformSection
      channels [
       Location.X {type=MovieSceneScriptingDoubleChannel, default=7265.8323, const [7200.8853,1]}
       Location.Y {type=MovieSceneScriptingDoubleChannel, default=11572.4742, const [11571.4629,1]}
       Location.Z {type=MovieSceneScriptingDoubleChannel, default=394.9222, const [371.7595,1]}
       Rotation.X {type=MovieSceneScriptingDoubleChannel, default=0, const [0,1]}
       Rotation.Y {type=MovieSceneScriptingDoubleChannel, default=0, const [0,1]}
       Rotation.Z {type=MovieSceneScriptingDoubleChannel, default=0, const [0,1]}
       Scale.X {type=MovieSceneScriptingDoubleChannel, default=1, const [1,1]}
       Scale.Y {type=MovieSceneScriptingDoubleChannel, default=1, const [1,1]}
       Scale.Z {type=MovieSceneScriptingDoubleChannel, default=1, const [1,1]}
      ]
     }]
   }]
 }
 TestScaleEditorActor2 {
  id=#14
  possessed_class=TestScaleEditorActor_C
  tracks [{
    type=MovieSceneVisibilityTrack
    sections [{type=MovieSceneVisibilitySection, channels [None {type=MovieSceneScriptingBoolChannel, default=false}]}]
   }]
 }
 DoorComponent {
  id=#15
  parent_id=#16
  possessed_class=FortBuildingWallDoorComponent
  tracks [
   {
    type=MovieScene3DTransformTrack
    sections [{
      type=MovieScene3DTransformSection
      channels [
       Location.X {type=MovieSceneScriptingDoubleChannel, default=0}
       Location.Y {type=MovieSceneScriptingDoubleChannel, default=-1}
       Location.Z {type=MovieSceneScriptingDoubleChannel, default=1}
       Rotation.X {type=MovieSceneScriptingDoubleChannel, default=0}
       Rotation.Y {type=MovieSceneScriptingDoubleChannel, default=0}
       Rotation.Z {type=MovieSceneScriptingDoubleChannel, default=0}
       Scale.X {type=MovieSceneScriptingDoubleChannel, default=1}
       Scale.Y {type=MovieSceneScriptingDoubleChannel, default=1}
       Scale.Z {type=MovieSceneScriptingDoubleChannel, default=1}
      ]
     }]
   }
   {
    type=MovieSceneVisibilityTrack
    sections [{type=MovieSceneVisibilitySection, channels [None {type=MovieSceneScriptingBoolChannel, default=false}]}]
   }
  ]
 }
 BrickSimple_Chimney_B {
  id=#17
  possessed_class=BP_WildEstate_House_Fireplace_A_C
  tracks [{
    type=MovieSceneVisibilityTrack
    sections [{type=MovieSceneVisibilitySection, channels [None {type=MovieSceneScriptingBoolChannel, default=true, keys [[1001,false],[1180,true]]}]}]
   }]
 }
 HoodLOD0_LOD0 {
  id=#18
  possessed_class=FortStaticMeshActor
  tracks [
   {
    type=MovieScene3DTransformTrack
    sections [{
      type=MovieScene3DTransformSection
      channels [
       Location.X {type=MovieSceneScriptingDoubleChannel, default=16952.6667}
       Location.Y {type=MovieSceneScriptingDoubleChannel, default=3160.7244}
       Location.Z {type=MovieSceneScriptingDoubleChannel, default=10.8123}
       Rotation.X {type=MovieSceneScriptingDoubleChannel, default=0}
       Rotation.Y {type=MovieSceneScriptingDoubleChannel, default=0}
       Rotation.Z {type=MovieSceneScriptingDoubleChannel, default=157.4188}
       Scale.X {type=MovieSceneScriptingDoubleChannel, default=1}
       Scale.Y {type=MovieSceneScriptingDoubleChannel, default=1}
       Scale.Z {type=MovieSceneScriptingDoubleChannel, default=1}
      ]
     }]
   }
   {
    type=MovieSceneVisibilityTrack
    sections [{type=MovieSceneVisibilitySection, channels [None {type=MovieSceneScriptingBoolChannel, default=false}]}]
   }
  ]
 }
 _1_BaseOrbit_CTRL {
  id=#19
  possessed_class=FortStaticMeshActor
  tracks [
   {
    type=MovieSceneSpawnTrack
    sections [{type=MovieSceneSpawnSection, channels [None {type=MovieSceneScriptingBoolChannel, default=true, const [true,1]}]}]
   }
   {
    type=MovieScene3DTransformTrack
    sections [{
      type=MovieScene3DTransformSection
      channels [
       Location.X {type=MovieSceneScriptingDoubleChannel, default=18.5133, const [0,3]}
       Location.Y {type=MovieSceneScriptingDoubleChannel, default=0, const [0,3]}
       Location.Z {type=MovieSceneScriptingDoubleChannel, default=0, const [0,3]}
       Rotation.X {type=MovieSceneScriptingDoubleChannel, default=0, const [0,19]}
       Rotation.Y {type=MovieSceneScriptingDoubleChannel, default=0, const [0,19]}
       Rotation.Z {
        type=MovieSceneScriptingDoubleChannel
        default=114.737
        keys [[1001,28.185],[1121,28.185],[1133,26.0622],[1160,4.3495],[1177,-33.0399],[1195,-61.3435],[1226,-82.4931],[1278,-104.1469],[1377,-127.5636],[1410,-120.4872],[1446,-84.7222],[1462,-86.1862],[1474,-84.64],[1502,-70.5175],[1586,-54.9767],[1724,-56.2848],[1737,-185.5326],[1749,-683.4629]]
       }
       Scale.X {type=MovieSceneScriptingDoubleChannel, default=1, const [1,19]}
       Scale.Y {type=MovieSceneScriptingDoubleChannel, default=1, const [1,19]}
       Scale.Z {type=MovieSceneScriptingDoubleChannel, default=1, const [1,19]}
      ]
     }]
   }
   {type=MovieScene3DAttachTrack, sections [{type=MovieScene3DAttachSection}]}
  ]
 }
 BP_Blockout_Doorway_A_A_A8 {
  id=#20
  possessed_class=BP_WildEstate_House_Window_C_A_C
  tracks [{
    type=MovieSceneVisibilityTrack
    sections [{type=MovieSceneVisibilitySection, channels [None {type=MovieSceneScriptingBoolChannel, default=true, keys [[1001,false],[1180,true]]}]}]
   }]
 }
 SpotLight {
  id=#12
  possessed_class=SpotLight
  tracks [{
    type=MovieScene3DTransformTrack
    sections [{
      type=MovieScene3DTransformSection
      channels [
       Location.X {type=MovieSceneScriptingDoubleChannel, default=7264, const [7240.2725,1]}
       Location.Y {type=MovieSceneScriptingDoubleChannel, default=11552, const [11608.0605,1]}
       Location.Z {type=MovieSceneScriptingDoubleChannel, default=466, const [440.0335,1]}
       Rotation.X {type=MovieSceneScriptingDoubleChannel, default=-90, const [-90,1]}
       Rotation.Y {type=MovieSceneScriptingDoubleChannel, default=-2.8125, const [-2.8125,1]}
       Rotation.Z {type=MovieSceneScriptingDoubleChannel, default=90, const [90,1]}
       Scale.X {type=MovieSceneScriptingDoubleChannel, default=1, const [1,1]}
       Scale.Y {type=MovieSceneScriptingDoubleChannel, default=1, const [1,1]}
       Scale.Z {type=MovieSceneScriptingDoubleChannel, default=1, const [1,1]}
      ]
     }]
   }]
 }
 _3_TiltRoll_CTRL {
  id=#21
  possessed_class=FortStaticMeshActor
  tracks [
   {
    type=MovieSceneSpawnTrack
    sections [{type=MovieSceneSpawnSection, channels [None {type=MovieSceneScriptingBoolChannel, default=true, const [true,2]}]}]
   }
   {
    type=MovieScene3DTransformTrack
    sections [{
      type=MovieScene3DTransformSection
      channels [
       Location.X {type=MovieSceneScriptingDoubleChannel, default=0, thinned_from=13, keys [[1001,-55.3521],[1121,-0.8799],[1133,-0.0058],[1278,0],[1923,0]]}
       Location.Y {type=MovieSceneScriptingDoubleChannel, default=0, thinned_from=13, keys [[1001,0],[1923,0]]}
       Location.Z {type=MovieSceneScriptingDoubleChannel, default=0, thinned_from=13, keys [[1001,9.4262],[1121,0.1498],[1133,0.001],[1278,0],[1923,0]]}
       Rotation.X {type=MovieSceneScriptingDoubleChannel, default=0, thinned_from=13, keys [[1001,0],[1923,0]]}
       Rotation.Y {
        type=MovieSceneScriptingDoubleChannel
        default=-7.6791
        thinned_from=14
        keys [[1001,-9.6645],[1121,-9.6645],[1133,-8.9973],[1278,7.0985],[1430,7.0985],[1436,-2.421],[1447,-11.0251],[1463,-9.7436],[1509,7.0985],[1551,0.5717],[1569,-4.1475],[1586,-6.2743],[1923,-6.4644]]
       }
       Rotation.Z {type=MovieSceneScriptingDoubleChannel, default=0, thinned_from=13, keys [[1001,0],[1923,0]]}
       Scale.X {type=MovieSceneScriptingDoubleChannel, default=1, const [1,13]}
       Scale.Y {type=MovieSceneScriptingDoubleChannel, default=1, const [1,13]}
       Scale.Z {type=MovieSceneScriptingDoubleChannel, default=1, const [1,13]}
      ]
     }]
   }
   {type=MovieScene3DAttachTrack, sections [{type=MovieScene3DAttachSection}]}
  ]
 }
 CameraComponent {
  id=#22
  parent_id=#4
  possessed_class=CineCameraComponent
  tracks [
   {
    type=MovieSceneFloatTrack
    display_name=Current Focal Length
    sections [{
      type=MovieSceneFloatSection
      channels [None {
        type=MovieSceneScriptingFloatChannel
        default=38
        thinned_from=11
        keys [[1001,38],[1121,38],[1160,33.3745],[1169,27.175],[1230,17],[1402,17],[1430,14.839],[1451,7.6963],[1477,5.5121],[1589,28]]
       }]
     }]
   }
   {
    type=MovieSceneFloatTrack
    display_name="Manual Focus Distance (Focus Settings)"
    sections [{type=MovieSceneFloatSection, channels [None {type=MovieSceneScriptingFloatChannel, default=0, const [0,3]}]}]
   }
   {
    type=MovieSceneFloatTrack
    display_name=Current Aperture
    sections [{type=MovieSceneFloatSection, channels [None {type=MovieSceneScriptingFloatChannel, default=2.8, const [2.8,2]}]}]
   }
   {
    type=MovieScene3DTransformTrack
    sections [{
      type=MovieScene3DTransformSection
      channels [
       Location.X {type=MovieSceneScriptingDoubleChannel, default=0, const [0,3]}
       Location.Y {type=MovieSceneScriptingDoubleChannel, default=0, const [0,3]}
       Location.Z {type=MovieSceneScriptingDoubleChannel, default=0, const [0,3]}
       Rotation.X {type=MovieSceneScriptingDoubleChannel, default=0, const [0,3]}
       Rotation.Y {type=MovieSceneScriptingDoubleChannel, default=0, const [0,3]}
       Rotation.Z {type=MovieSceneScriptingDoubleChannel, default=0, const [0,3]}
       Scale.X {type=MovieSceneScriptingDoubleChannel, default=1, const [1,3]}
       Scale.Y {type=MovieSceneScriptingDoubleChannel, default=1, const [1,3]}
       Scale.Z {type=MovieSceneScriptingDoubleChannel, default=1, const [1,3]}
      ]
     }]
   }
  ]
 }
 _2_PanBoom_CTRL {
  id=#23
  possessed_class=FortStaticMeshActor
  tracks [
   {
    type=MovieSceneSpawnTrack
    sections [{type=MovieSceneSpawnSection, channels [None {type=MovieSceneScriptingBoolChannel, default=true, const [true,1]}]}]
   }
   {
    type=MovieScene3DTransformTrack
    sections [{
      type=MovieScene3DTransformSection
      channels [
       Location.X {
        type=MovieSceneScriptingDoubleChannel
        default=-806.4399
        keys [[1001,19.5913],[1121,19.5913],[1133,17.6886],[1177,2.7086],[1278,-27.8719],[1399,-43.4963],[1461,-21.1465],[1509,-5.6911],[1586,0],[1724,-40.4706]]
       }
       Location.Y {
        type=MovieSceneScriptingDoubleChannel
        default=-7.2342
        thinned_from=11
        keys [[1001,-123.9342],[1121,-123.9342],[1133,-113.6576],[1161,-9.1734],[1278,0],[1399,0],[1447,-1.1471],[1474,0],[1724,0]]
       }
       Location.Z {
        type=MovieSceneScriptingDoubleChannel
        default=252.6591
        keys [[1001,198.6771],[1121,198.6771],[1133,191.4495],[1195,118.3018],[1278,62.1198],[1436,11.3668],[1479,32.6016],[1509,76.7968],[1552,532.2726],[1570,848.0606],[1586,965.1935],[1724,987.6728]]
       }
       Rotation.X {type=MovieSceneScriptingDoubleChannel, default=0, const [0,12]}
       Rotation.Y {type=MovieSceneScriptingDoubleChannel, default=0, const [0,12]}
       Rotation.Z {
        type=MovieSceneScriptingDoubleChannel
        default=-0.582
        thinned_from=11
        keys [[1001,-1.1975],[1121,-1.1975],[1133,-1.0064],[1161,0],[1724,0]]
       }
       Scale.X {type=MovieSceneScriptingDoubleChannel, default=1, const [1,12]}
       Scale.Y {type=MovieSceneScriptingDoubleChannel, default=1, const [1,12]}
       Scale.Z {type=MovieSceneScriptingDoubleChannel, default=1, const [1,12]}
      ]
     }]
   }
   {type=MovieScene3DAttachTrack, sections [{type=MovieScene3DAttachSection}]}
  ]
 }
 CameraRig_Rail {
  id=#24
  possessed_class=CameraRig_Rail
  tracks [
   {
    type=MovieSceneSpawnTrack
    sections [{type=MovieSceneSpawnSection, channels [None {type=MovieSceneScriptingBoolChannel, default=true, const [true,2]}]}]
   }
   {
    type=MovieScene3DTransformTrack
    sections [{
      type=MovieScene3DTransformSection
      channels [
       Location.X {type=MovieSceneScriptingDoubleChannel, default=7464.3386, thinned_from=4, keys [[1001,7464.3386],[1444,7464.3384]]}
       Location.Y {type=MovieSceneScriptingDoubleChannel, default=11012.2387, thinned_from=4, keys [[1001,11012.2387],[1444,11012.2383]]}
       Location.Z {type=MovieSceneScriptingDoubleChannel, default=422.674, thinned_from=4, keys [[1003,422.674],[1444,422.674]]}
       Rotation.X {type=MovieSceneScriptingDoubleChannel, default=0, const [0,4]}
       Rotation.Y {type=MovieSceneScriptingDoubleChannel, default=0, const [0,4]}
       Rotation.Z {type=MovieSceneScriptingDoubleChannel, default=86.8822, thinned_from=4, keys [[1003,86.8822],[1444,86.8822]]}
       Scale.X {type=MovieSceneScriptingDoubleChannel, default=1, const [1,4]}
       Scale.Y {type=MovieSceneScriptingDoubleChannel, default=1, const [1,4]}
       Scale.Z {type=MovieSceneScriptingDoubleChannel, default=1, const [1,4]}
      ]
     }]
   }
   {
    type=MovieSceneFloatTrack
    display_name=Current Position on Rail
    sections [{
      type=MovieSceneFloatSection
      channels [None {
        type=MovieSceneScriptingFloatChannel
        default=0.1026
        keys [[1001,0],[1121,0],[1130,0.0156],[1163,0.1026],[1226,0.1781],[1375,0.2354],[1435,0.3039],[1483,0.4855],[1502,0.6107],[1556,0.9218],[1586,0.9773],[1724,0.9928]]
       }]
     }]
   }
  ]
 }
 BP_Blockout_Doorway_A_A_A17 {
  id=#25
  possessed_class=BP_WildEstate_Blockout_Wall_A_A_C
  tracks [{
    type=MovieSceneVisibilityTrack
    sections [{
      type=MovieSceneVisibilitySection
      channels [None {type=MovieSceneScriptingBoolChannel, default=false, keys [[1001,true],[1180,false],[1379,true]]}]
     }]
   }]
 }
 BP_Blockout_Doorway_A_A_A {id=#16, possessed_class=BP_WildEstate_House_Door_A_A_C, tracks []}
 Artemis_Suburban_CurtainsModern_C3 {
  id=#26
  possessed_class=Artemis_Suburban_CurtainsModern_C_C
  tracks [{
    type=MovieSceneVisibilityTrack
    sections [{type=MovieSceneVisibilitySection, channels [None {type=MovieSceneScriptingBoolChannel, default=true, keys [[1001,false],[1180,true]]}]}]
   }]
 }
 door_pivot {
  id=#27
  possessed_class=FortStaticMeshActor
  tracks [
   {type=MovieSceneSpawnTrack, sections [{type=MovieSceneSpawnSection, channels [None {type=MovieSceneScriptingBoolChannel, default=true}]}]}
   {
    type=MovieScene3DTransformTrack
    sections [{
      type=MovieScene3DTransformSection
      channels [
       Location.X {type=MovieSceneScriptingDoubleChannel, default=6400.9955, const [6400.9956,6]}
       Location.Y {type=MovieSceneScriptingDoubleChannel, default=12288, const [12373.583,6]}
       Location.Z {type=MovieSceneScriptingDoubleChannel, default=384.8292, const [384.8292,6]}
       Rotation.X {type=MovieSceneScriptingDoubleChannel, default=0, const [0,6]}
       Rotation.Y {type=MovieSceneScriptingDoubleChannel, default=0, const [0,6]}
       Rotation.Z {type=MovieSceneScriptingDoubleChannel, default=90, keys [[1458,90],[1464,142.2273],[1476,183.515],[1519,183.515],[1526,155.355],[1539,90]]}
       Scale.X {type=MovieSceneScriptingDoubleChannel, default=1, const [1,6]}
       Scale.Y {type=MovieSceneScriptingDoubleChannel, default=1, const [1,6]}
       Scale.Z {type=MovieSceneScriptingDoubleChannel, default=1, const [1,6]}
      ]
     }]
   }
  ]
 }
 B_BattleBus_Stamina {
  id=#28
  possessed_class=B_BattleBus_Stamina_C
  tracks [
   {type=MovieSceneSpawnTrack, sections [{type=MovieSceneSpawnSection, channels [None {type=MovieSceneScriptingBoolChannel, default=true}]}]}
   {
    type=MovieScene3DTransformTrack
    sections [{
      type=MovieScene3DTransformSection
      channels [
       Location.X {type=MovieSceneScriptingDoubleChannel, default=-19250.0933, const [-19250.0938,1]}
       Location.Y {type=MovieSceneScriptingDoubleChannel, default=-25564.6301, const [-25564.6309,1]}
       Location.Z {type=MovieSceneScriptingDoubleChannel, default=38281.5445, const [38281.543,1]}
       Rotation.X {type=MovieSceneScriptingDoubleChannel, default=0, const [0,1]}
       Rotation.Y {type=MovieSceneScriptingDoubleChannel, default=0, const [0,1]}
       Rotation.Z {type=MovieSceneScriptingDoubleChannel, default=44.2791, const [44.2791,1]}
       Scale.X {type=MovieSceneScriptingDoubleChannel, default=1, const [1,1]}
       Scale.Y {type=MovieSceneScriptingDoubleChannel, default=1, const [1,1]}
       Scale.Z {type=MovieSceneScriptingDoubleChannel, default=1, const [1,1]}
      ]
     }]
   }
   {
    type=MovieSceneVisibilityTrack
    sections [{type=MovieSceneVisibilitySection, channels [None {type=MovieSceneScriptingBoolChannel, default=true, keys [[1001,false],[2363,true]]}]}]
   }
  ]
 }
 SM_WildEstate_House_PictureFrames_A {
  id=#29
  possessed_class=BP_WildEstate_House_PictureFrames_A_C
  tracks [{
    type=MovieSceneVisibilityTrack
    sections [{type=MovieSceneVisibilitySection, channels [None {type=MovieSceneScriptingBoolChannel, default=true, keys [[1001,false],[1180,true]]}]}]
   }]
 }
]
display_rate {denominator=1, numerator=60}
tick_resolution {denominator=1, numerator=24000}
tracks [{type=MovieSceneCameraCutTrack, sections [{type=MovieSceneCameraCutSection, end_frame=2884, start_frame=1001}]}]
