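# Level Sequence for WildEstate Pre shot A0010, a moving camera shot around the estate house with a late Battle Bus pass.
# Cam_A0010 rides a layered rig: animated rail position plus attached orbit, pan/boom and tilt/roll control meshes, ending in a steep rise.
# Visibility keys swap walls and props around the camera move, while a pivot swings a door and a TV and spotlight shift.
# A second cine camera, camera cut and time warp track are included.
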LevelSequence /WildEstate/Sequences/Pre/A0010/WildEstate_Pre_A0010_04
bindings [
 BP_Blockout_Doorway_A_A_A75 {
  id=#1
  possessed_class=BP_WildEstate_Blockout_Doorway_Side_A_B_C
  tracks [{
    type=MovieSceneVisibilityTrack
    sections [{
      type=MovieSceneVisibilitySection
      channels [None {type=MovieSceneScriptingBoolChannel, default=false, keys [[1001,true],[1356,false],[1722,true]]}]
     }]
   }]
 }
 TestScaleEditorActor3 {
  id=#2
  possessed_class=TestScaleEditorActor_C
  tracks [{
    type=MovieSceneVisibilityTrack
    sections [{type=MovieSceneVisibilitySection, channels [None {type=MovieSceneScriptingBoolChannel, default=false}]}]
   }]
 }
 _A0010_WorldCube {
  id=#3
  possessed_class=FortStaticMeshActor
  tracks [
   {type=MovieSceneSpawnTrack, sections [{type=MovieSceneSpawnSection, channels [None {type=MovieSceneScriptingBoolChannel, default=true}]}]}
   {
    type=MovieScene3DTransformTrack
    sections [{
      type=MovieScene3DTransformSection
      channels [
       Location.X {type=MovieSceneScriptingDoubleChannel, default=7205.2965}
       Location.Y {type=MovieSceneScriptingDoubleChannel, default=11814.1699}
       Location.Z {type=MovieSceneScriptingDoubleChannel, default=366.018}
       Rotation.X {type=MovieSceneScriptingDoubleChannel, default=0}
       Rotation.Y {type=MovieSceneScriptingDoubleChannel, default=0}
       Rotation.Z {type=MovieSceneScriptingDoubleChannel, default=0}
       Scale.X {type=MovieSceneScriptingDoubleChannel, default=1}
       Scale.Y {type=MovieSceneScriptingDoubleChannel, default=1}
       Scale.Z {type=MovieSceneScriptingDoubleChannel, default=1}
      ]
     }]
   }
  ]
 }
 Cam_A0010 {
  id=#4
  possessed_class=CineCameraActor
  tracks [
   {
    type=MovieScene3DTransformTrack
    sections [{
      type=MovieScene3DTransformSection
      channels [
       Location.X {type=MovieSceneScriptingDoubleChannel, default=0, const [0,2]}
       Location.Y {type=MovieSceneScriptingDoubleChannel, default=0, const [0,2]}
       Location.Z {type=MovieSceneScriptingDoubleChannel, default=0, const [0,2]}
       Rotation.X {type=MovieSceneScriptingDoubleChannel, default=0, const [0,2]}
       Rotation.Y {type=MovieSceneScriptingDoubleChannel, default=0, const [0,2]}
       Rotation.Z {type=MovieSceneScriptingDoubleChannel, default=0, const [0,2]}
       Scale.X {type=MovieSceneScriptingDoubleChannel, default=1, const [1,2]}
       Scale.Y {type=MovieSceneScriptingDoubleChannel, default=1, const [1,2]}
       Scale.Z {type=MovieSceneScriptingDoubleChannel, default=1, const [1,2]}
      ]
     }]
   }
   {type=MovieSceneSpawnTrack, sections [{type=MovieSceneSpawnSection, channels [None {type=MovieSceneScriptingBoolChannel, default=true}]}]}
   {type=MovieScene3DAttachTrack, sections [{type=MovieScene3DAttachSection}]}
  ]
 }
 CineCameraActor {
  id=#5
  possessed_class=CineCameraActor
  tracks [
   {
    type=MovieScene3DTransformTrack
    sections [{
      type=MovieScene3DTransformSection
      channels [
       Location.X {type=MovieSceneScriptingDoubleChannel, default=7567.6929, const [7567.6929,1]}
       Location.Y {type=MovieSceneScriptingDoubleChannel, default=11083.6008, const [11083.6006,1]}
       Location.Z {type=MovieSceneScriptingDoubleChannel, default=621.1719, const [621.1719,1]}
       Rotation.X {type=MovieSceneScriptingDoubleChannel, default=0, const [0,1]}
       Rotation.Y {type=MovieSceneScriptingDoubleChannel, default=-9.4, const [-9.4,1]}
       Rotation.Z {type=MovieSceneScriptingDoubleChannel, default=114.4, const [114.4,1]}
       Scale.X {type=MovieSceneScriptingDoubleChannel, default=1, const [1,1]}
       Scale.Y {type=MovieSceneScriptingDoubleChannel, default=1, const [1,1]}
       Scale.Z {type=MovieSceneScriptingDoubleChannel, default=1, const [1,1]}
      ]
     }]
   }
   {type=MovieSceneSpawnTrack, sections [{type=MovieSceneSpawnSection, channels [None {type=MovieSceneScriptingBoolChannel, default=true}]}]}
  ]
 }
 GeoScript_WildEstate_VTXColor2 {
  id=#6
  possessed_class=GeoScript_WildEstate_VTXColor_C
  tracks [{
    type=MovieSceneVisibilityTrack
    sections [{type=MovieSceneVisibilitySection, channels [None {type=MovieSceneScriptingBoolChannel, default=false}]}]
   }]
 }
 BP_Blockout_Stairs_Rail_A_TESTTT {
  id=#7
  possessed_class=BP_Estate_Stairs_Rail_A_B_C
  tracks [{
    type=MovieSceneVisibilityTrack
    sections [{
      type=MovieSceneVisibilitySection
      channels [None {type=MovieSceneScriptingBoolChannel, default=false, keys [[1001,true],[1376,false],[1722,true]]}]
     }]
   }]
 }
 BP_Blockout_Doorway_A_A_A9 {
  id=#8
  possessed_class=BP_WildEstate_Blockout_Wall_A_A_C
  tracks [{
    type=MovieSceneVisibilityTrack
    sections [{type=MovieSceneVisibilitySection, channels [None {type=MovieSceneScriptingBoolChannel, default=true, keys [[1001,false],[1722,true]]}]}]
   }]
 }
 CameraComponent {
  id=#9
  parent_id=#5
  possessed_class=CineCameraComponent
  tracks [
   {
    type=MovieSceneFloatTrack
    display_name=Current Focal Length
    sections [{
      type=MovieSceneFloatSection
      channels [None {type=MovieSceneScriptingFloatChannel, default=14.59, thinned_from=5, keys [[1151,38],[1191,14],[1993,14],[2042,14.59]]}]
     }]
   }
   {
    type=MovieSceneFloatTrack
    display_name="Manual Focus Distance (Focus Settings)"
    sections [{type=MovieSceneFloatSection, channels [None {type=MovieSceneScriptingFloatChannel, default=100000}]}]
   }
   {
    type=MovieSceneFloatTrack
    display_name=Current Aperture
    sections [{type=MovieSceneFloatSection, channels [None {type=MovieSceneScriptingFloatChannel, default=2.8}]}]
   }
   {
    type=MovieScene3DTransformTrack
    sections [{
      type=MovieScene3DTransformSection
      channels [
       Location.X {type=MovieSceneScriptingDoubleChannel, default=0, const [0,1]}
       Location.Y {type=MovieSceneScriptingDoubleChannel, default=58.2237, const [0,1]}
       Location.Z {type=MovieSceneScriptingDoubleChannel, default=0, const [0,1]}
       Rotation.X {type=MovieSceneScriptingDoubleChannel, default=0, const [0,1]}
       Rotation.Y {type=MovieSceneScriptingDoubleChannel, default=0, const [0,1]}
       Rotation.Z {type=MovieSceneScriptingDoubleChannel, default=0, const [0,1]}
       Scale.X {type=MovieSceneScriptingDoubleChannel, default=1, const [1,1]}
       Scale.Y {type=MovieSceneScriptingDoubleChannel, default=1, const [1,1]}
       Scale.Z {type=MovieSceneScriptingDoubleChannel, default=1, const [1,1]}
      ]
     }]
   }
  ]
 }
 SM_WildEstate_House_Door_A {
  id=#10
  possessed_class=FortStaticMeshActor
  tracks [
   {type=MovieSceneSpawnTrack, sections [{type=MovieSceneSpawnSection, channels [None {type=MovieSceneScriptingBoolChannel, default=true}]}]}
   {
    type=MovieScene3DTransformTrack
    sections [{
      type=MovieScene3DTransformSection
      channels [
       Location.X {type=MovieSceneScriptingDoubleChannel, default=-85.583}
       Location.Y {type=MovieSceneScriptingDoubleChannel, default=0}
       Location.Z {type=MovieSceneScriptingDoubleChannel, default=0}
       Rotation.X {type=MovieSceneScriptingDoubleChannel, default=0}
       Rotation.Y {type=MovieSceneScriptingDoubleChannel, default=0}
       Rotation.Z {type=MovieSceneScriptingDoubleChannel, default=0}
       Scale.X {type=MovieSceneScriptingDoubleChannel, default=1}
       Scale.Y {type=MovieSceneScriptingDoubleChannel, default=1}
       Scale.Z {type=MovieSceneScriptingDoubleChannel, default=1}
      ]
     }]
   }
   {type=MovieScene3DAttachTrack, sections [{type=MovieScene3DAttachSection}]}
  ]
 }
 BP_Blockout_Doorway_A_A_A73 {
  id=#11
  possessed_class=BP_WildEstate_Blockout_Doorway_Side_A_B_C
  tracks [{
    type=MovieSceneVisibilityTrack
    sections [{
      type=MovieSceneVisibilitySection
      channels [None {type=MovieSceneScriptingBoolChannel, default=false, keys [[1001,true],[1384,false],[1722,true]]}]
     }]
   }]
 }
 LightComponent0 {
  id=#12
  parent_id=#13
  possessed_class=SpotLightComponent
  tracks [
   {
    type=MovieSceneFloatTrack
    display_name=Intensity
    sections [{type=MovieSceneFloatSection, channels [None {type=MovieSceneScriptingFloatChannel, default=4}]}]
   }
   {
    type=MovieSceneColorTrack
    display_name=Light Color
    sections [{
      type=MovieSceneColorSection
      channels [
       Color.R {type=MovieSceneScriptingFloatChannel, default=1}
       Color.G {type=MovieSceneScriptingFloatChannel, default=1}
       Color.B {type=MovieSceneScriptingFloatChannel, default=1}
       Color.A {type=MovieSceneScriptingFloatChannel, default=1}
      ]
     }]
   }
  ]
 }
 BP_WildEstate_House_TV_A {
  id=#14
  possessed_class=BP_WildEstate_House_TV_A_C
  tracks [{
    type=MovieScene3DTransformTrack
    sections [{
      type=MovieScene3DTransformSection
      channels [
       Location.X {type=MovieSceneScriptingDoubleChannel, default=7265.8323, keys [[1001,7220.3354],[1439,7265.832]]}
       Location.Y {type=MovieSceneScriptingDoubleChannel, default=11572.4742, keys [[1001,11541.8691],[1439,11560.4746]]}
       Location.Z {type=MovieSceneScriptingDoubleChannel, default=394.9222, keys [[1001,378.3755],[1439,394.9221]]}
       Rotation.X {type=MovieSceneScriptingDoubleChannel, default=0, const [0,2]}
       Rotation.Y {type=MovieSceneScriptingDoubleChannel, default=0, const [0,2]}
       Rotation.Z {type=MovieSceneScriptingDoubleChannel, default=0, const [0,2]}
       Scale.X {type=MovieSceneScriptingDoubleChannel, default=1, const [1,2]}
       Scale.Y {type=MovieSceneScriptingDoubleChannel, default=1, const [1,2]}
       Scale.Z {type=MovieSceneScriptingDoubleChannel, default=1, const [1,2]}
      ]
     }]
   }]
 }
 TestScaleEditorActor2 {
  id=#15
  possessed_class=TestScaleEditorActor_C
  tracks [{
    type=MovieSceneVisibilityTrack
    sections [{type=MovieSceneVisibilitySection, channels [None {type=MovieSceneScriptingBoolChannel, default=false}]}]
   }]
 }
 DoorComponent {
  id=#16
  parent_id=#17
  possessed_class=FortBuildingWallDoorComponent
  tracks [
   {
    type=MovieScene3DTransformTrack
    sections [{
      type=MovieScene3DTransformSection
      channels [
       Location.X {type=MovieSceneScriptingDoubleChannel, default=0}
       Location.Y {type=MovieSceneScriptingDoubleChannel, default=-1}
       Location.Z {type=MovieSceneScriptingDoubleChannel, default=1}
       Rotation.X {type=MovieSceneScriptingDoubleChannel, default=0}
       Rotation.Y {type=MovieSceneScriptingDoubleChannel, default=0}
       Rotation.Z {type=MovieSceneScriptingDoubleChannel, default=0}
       Scale.X {type=MovieSceneScriptingDoubleChannel, default=1}
       Scale.Y {type=MovieSceneScriptingDoubleChannel, default=1}
       Scale.Z {type=MovieSceneScriptingDoubleChannel, default=1}
      ]
     }]
   }
   {
    type=MovieSceneVisibilityTrack
    sections [{type=MovieSceneVisibilitySection, channels [None {type=MovieSceneScriptingBoolChannel, default=false}]}]
   }
  ]
 }
 BrickSimple_Chimney_B {
  id=#18
  possessed_class=BP_WildEstate_House_Fireplace_A_C
  tracks [{
    type=MovieSceneVisibilityTrack
    sections [{type=MovieSceneVisibilitySection, channels [None {type=MovieSceneScriptingBoolChannel, default=true, keys [[1001,false],[1384,true]]}]}]
   }]
 }
 HoodLOD0_LOD0 {
  id=#19
  possessed_class=FortStaticMeshActor
  tracks [
   {
    type=MovieScene3DTransformTrack
    sections [{
      type=MovieScene3DTransformSection
      channels [
       Location.X {type=MovieSceneScriptingDoubleChannel, default=16952.6667}
       Location.Y {type=MovieSceneScriptingDoubleChannel, default=3160.7244}
       Location.Z {type=MovieSceneScriptingDoubleChannel, default=10.8123}
       Rotation.X {type=MovieSceneScriptingDoubleChannel, default=0}
       Rotation.Y {type=MovieSceneScriptingDoubleChannel, default=0}
       Rotation.Z {type=MovieSceneScriptingDoubleChannel, default=157.4188}
       Scale.X {type=MovieSceneScriptingDoubleChannel, default=1}
       Scale.Y {type=MovieSceneScriptingDoubleChannel, default=1}
       Scale.Z {type=MovieSceneScriptingDoubleChannel, default=1}
      ]
     }]
   }
   {
    type=MovieSceneVisibilityTrack
    sections [{type=MovieSceneVisibilitySection, channels [None {type=MovieSceneScriptingBoolChannel, default=false}]}]
   }
  ]
 }
 _1_BaseOrbit_CTRL {
  id=#20
  possessed_class=FortStaticMeshActor
  tracks [
   {type=MovieSceneSpawnTrack, sections [{type=MovieSceneSpawnSection, channels [None {type=MovieSceneScriptingBoolChannel, default=true}]}]}
   {
    type=MovieScene3DTransformTrack
    sections [{
      type=MovieScene3DTransformSection
      channels [
       Location.X {type=MovieSceneScriptingDoubleChannel, default=18.5133, const [0,24]}
       Location.Y {type=MovieSceneScriptingDoubleChannel, default=0, const [0,24]}
       Location.Z {type=MovieSceneScriptingDoubleChannel, default=0, const [0,24]}
       Rotation.X {type=MovieSceneScriptingDoubleChannel, default=0, thinned_from=21, keys [[1001,0],[6353,0]]}
       Rotation.Y {type=MovieSceneScriptingDoubleChannel, default=0, thinned_from=21, keys [[1001,0],[6353,0]]}
       Rotation.Z {
        type=MovieSceneScriptingDoubleChannel
        default=114.737
        thinned_from=36
        keys [[1001,28.6135],[1121,28.6135],[1140,27.5925],[1261,8.2298],[1341,-15.1319],[1350.0025,-19.0056],[1418,-52.5661],[1462,-68.7569],[1544,-96.1077],[1639,-117.708],[1759,-123.8686],[1850,-161.5789],[1914,-266.5968],[1949,-274.96],[2023,-274.96],[2417,-409.9868],[2542,-413.5393],[2700,-451.9748],[2742,-457.1745],[2898,-462.2159],[3012,-447.4994],[3099,-447.4994],[3357,-424.6546],[3836,-409.532],[3952,-415.6965],[4101,-435.4286],[4280,-461.8914],[4413,-484.5366],[4487,-516.0155],[4678,-569.4584],[4983,-586.8074],[5186,-519.8899],[5350,-484.7694],[5506,-473.7233],[6353,-421.8441]]
       }
       Scale.X {type=MovieSceneScriptingDoubleChannel, default=1, const [1,24]}
       Scale.Y {type=MovieSceneScriptingDoubleChannel, default=1, const [1,24]}
       Scale.Z {type=MovieSceneScriptingDoubleChannel, default=1, const [1,24]}
      ]
     }]
   }
   {type=MovieScene3DAttachTrack, sections [{type=MovieScene3DAttachSection}]}
  ]
 }
 BP_Blockout_Doorway_A_A_A8 {
  id=#21
  possessed_class=BP_WildEstate_House_Window_C_A_C
  tracks [{
    type=MovieSceneVisibilityTrack
    sections [{type=MovieSceneVisibilitySection, channels [None {type=MovieSceneScriptingBoolChannel, default=true, keys [[1001,false],[1722,true]]}]}]
   }]
 }
 SpotLight {
  id=#13
  possessed_class=SpotLight
  tracks [{
    type=MovieScene3DTransformTrack
    sections [{
      type=MovieScene3DTransformSection
      channels [
       Location.X {type=MovieSceneScriptingDoubleChannel, default=7264, keys [[1001,7240.2725],[1439,7264]]}
       Location.Y {type=MovieSceneScriptingDoubleChannel, default=11552, keys [[1001,11608.0605],[1439,11572]]}
       Location.Z {type=MovieSceneScriptingDoubleChannel, default=466, keys [[1001,440.0335],[1439,466]]}
       Rotation.X {type=MovieSceneScriptingDoubleChannel, default=-90, const [-90,2]}
       Rotation.Y {type=MovieSceneScriptingDoubleChannel, default=-2.8125, const [-2.8125,2]}
       Rotation.Z {type=MovieSceneScriptingDoubleChannel, default=90, const [90,2]}
       Scale.X {type=MovieSceneScriptingDoubleChannel, default=1, const [1,2]}
       Scale.Y {type=MovieSceneScriptingDoubleChannel, default=1, const [1,2]}
       Scale.Z {type=MovieSceneScriptingDoubleChannel, default=1, const [1,2]}
      ]
     }]
   }]
 }
 _3_TiltRoll_CTRL {
  id=#22
  possessed_class=FortStaticMeshActor
  tracks [
   {type=MovieSceneSpawnTrack, sections [{type=MovieSceneSpawnSection, channels [None {type=MovieSceneScriptingBoolChannel, default=true}]}]}
   {
    type=MovieScene3DTransformTrack
    sections [{
      type=MovieScene3DTransformSection
      channels [
       Location.X {type=MovieSceneScriptingDoubleChannel, default=0, thinned_from=22, keys [[1001,-55.3521],[1121,-0.8799],[6353,-0.8799]]}
       Location.Y {type=MovieSceneScriptingDoubleChannel, default=0, thinned_from=22, keys [[1001,0],[6353,0]]}
       Location.Z {type=MovieSceneScriptingDoubleChannel, default=0, thinned_from=22, keys [[1001,9.4262],[1121,0.1498],[6353,0.1498]]}
       Rotation.X {type=MovieSceneScriptingDoubleChannel, default=0, thinned_from=22, keys [[1001,0],[6353,0]]}
       Rotation.Y {
        type=MovieSceneScriptingDoubleChannel
        default=-7.6791
        keys [[1001,-9.6645],[1121,-9.6645],[1261,-5.2708],[1350,-2.8605],[1924,-2.8605],[2172,-1.2971],[2416,1.0649],[2542,1.5218],[2740,-2.8605],[2742,-2.8605],[3099,-1.4634],[3321,0.9858],[3353,1.4041],[3561,8.9438],[3737,18.494],[3870,7.3868],[3985,5.3245],[4446,4.5155],[4803,4.8014],[4983,1.0032],[5377,-26.1148],[5506,-24.8796],[5660,-22.0651],[6353,-18.8495]]
       }
       Rotation.Z {type=MovieSceneScriptingDoubleChannel, default=0, thinned_from=22, keys [[1001,0],[6353,0]]}
       Scale.X {type=MovieSceneScriptingDoubleChannel, default=1, const [1,22]}
       Scale.Y {type=MovieSceneScriptingDoubleChannel, default=1, const [1,22]}
       Scale.Z {type=MovieSceneScriptingDoubleChannel, default=1, const [1,22]}
      ]
     }]
   }
   {type=MovieScene3DAttachTrack, sections [{type=MovieScene3DAttachSection}]}
  ]
 }
 CameraComponent {
  id=#23
  parent_id=#4
  possessed_class=CineCameraComponent
  tracks [
   {
    type=MovieSceneFloatTrack
    display_name=Current Focal Length
    sections [{
      type=MovieSceneFloatSection
      channels [None {type=MovieSceneScriptingFloatChannel, default=24, keys [[1121,30],[1274,17.3164],[2069,17.2063],[2247,24]]}]
     }]
   }
   {
    type=MovieSceneFloatTrack
    display_name="Manual Focus Distance (Focus Settings)"
    sections [{type=MovieSceneFloatSection, channels [None {type=MovieSceneScriptingFloatChannel, default=0}]}]
   }
   {
    type=MovieSceneFloatTrack
    display_name=Current Aperture
    sections [{type=MovieSceneFloatSection, channels [None {type=MovieSceneScriptingFloatChannel, default=2.8}]}]
   }
   {
    type=MovieScene3DTransformTrack
    sections [{
      type=MovieScene3DTransformSection
      channels [
       Location.X {type=MovieSceneScriptingDoubleChannel, default=0, const [0,4]}
       Location.Y {type=MovieSceneScriptingDoubleChannel, default=0, const [0,4]}
       Location.Z {type=MovieSceneScriptingDoubleChannel, default=0, const [0,4]}
       Rotation.X {type=MovieSceneScriptingDoubleChannel, default=0, const [0,4]}
       Rotation.Y {type=MovieSceneScriptingDoubleChannel, default=0, const [0,4]}
       Rotation.Z {type=MovieSceneScriptingDoubleChannel, default=0, const [0,4]}
       Scale.X {type=MovieSceneScriptingDoubleChannel, default=1, const [1,4]}
       Scale.Y {type=MovieSceneScriptingDoubleChannel, default=1, const [1,4]}
       Scale.Z {type=MovieSceneScriptingDoubleChannel, default=1, const [1,4]}
      ]
     }]
   }
  ]
 }
 _2_PanBoom_CTRL {
  id=#24
  possessed_class=FortStaticMeshActor
  tracks [
   {type=MovieSceneSpawnTrack, sections [{type=MovieSceneSpawnSection, channels [None {type=MovieSceneScriptingBoolChannel, default=true}]}]}
   {
    type=MovieScene3DTransformTrack
    sections [{
      type=MovieScene3DTransformSection
      channels [
       Location.X {
        type=MovieSceneScriptingDoubleChannel
        default=-806.4399
        thinned_from=22
        keys [[1121,94.905],[1261,30.5877],[1350,0],[1554,-148.9901],[1697,-200.4992],[1924,-200.4992],[2172,-276.2491],[2416,-455.7071],[2542,-495.2287],[2740,-359.5419],[3099,-359.5419],[3321,0],[4803,-943.8052],[4894,-1004.9228],[4983,-1057.9316],[5257,-1152.2019],[5352,-1197.1572],[5506,-2854.8679],[6353,-83504.1399]]
       }
       Location.Y {
        type=MovieSceneScriptingDoubleChannel
        default=-7.2342
        thinned_from=26
        keys [[1121,-123.1737],[1261,-39.6986],[1350,0],[1845,0],[1924,31.3263],[1984,0],[2960,0],[3007,-45.6359],[3059,-42.6316],[3144,0],[6353,0]]
       }
       Location.Z {
        type=MovieSceneScriptingDoubleChannel
        default=252.6591
        thinned_from=21
        keys [[1121,204.4086],[1261,150.6534],[1350,111.0947],[1924,111.0947],[2172,185.4804],[2418,601.8976],[2542,610.9838],[2742,399.6054],[2960,164.577],[3007,140.4122],[3041,131.3377],[3321,106.094],[3836,106.522],[4803,106.522],[4983,416.0079],[5257,1339.356],[5352,1803.4069],[5506,3235.2732],[6353,38198.3217]]
       }
       Rotation.X {type=MovieSceneScriptingDoubleChannel, default=0, const [0,23]}
       Rotation.Y {type=MovieSceneScriptingDoubleChannel, default=0, const [0,23]}
       Rotation.Z {
        type=MovieSceneScriptingDoubleChannel
        default=-0.582
        thinned_from=23
        keys [[1121,0.5785],[5257,0.5785],[5352,2.5992],[5506,6.2733],[6353,0.5785]]
       }
       Scale.X {type=MovieSceneScriptingDoubleChannel, default=1, const [1,23]}
       Scale.Y {type=MovieSceneScriptingDoubleChannel, default=1, const [1,23]}
       Scale.Z {type=MovieSceneScriptingDoubleChannel, default=1, const [1,23]}
      ]
     }]
   }
   {type=MovieScene3DAttachTrack, sections [{type=MovieScene3DAttachSection}]}
  ]
 }
 BP_Blockout_Doorway_A_A_A17 {
  id=#25
  possessed_class=BP_WildEstate_Blockout_Wall_A_A_C
  tracks [{
    type=MovieSceneVisibilityTrack
    sections [{
      type=MovieSceneVisibilitySection
      channels [None {type=MovieSceneScriptingBoolChannel, default=false, keys [[1001,true],[1384,false],[1722,true]]}]
     }]
   }]
 }
 BP_Blockout_Doorway_A_A_A {id=#17, possessed_class=BP_WildEstate_House_Door_A_A_C, tracks []}
 Artemis_Suburban_CurtainsModern_C3 {
  id=#26
  possessed_class=Artemis_Suburban_CurtainsModern_C_C
  tracks [{
    type=MovieSceneVisibilityTrack
    sections [{type=MovieSceneVisibilitySection, channels [None {type=MovieSceneScriptingBoolChannel, default=true, keys [[1001,false],[1722,true]]}]}]
   }]
 }
 door_pivot {
  id=#27
  possessed_class=FortStaticMeshActor
  tracks [
   {type=MovieSceneSpawnTrack, sections [{type=MovieSceneSpawnSection, channels [None {type=MovieSceneScriptingBoolChannel, default=true}]}]}
   {
    type=MovieScene3DTransformTrack
    sections [{
      type=MovieScene3DTransformSection
      channels [
       Location.X {type=MovieSceneScriptingDoubleChannel, default=6400.9955, const [6400.9956,6]}
       Location.Y {type=MovieSceneScriptingDoubleChannel, default=12288, const [12373.583,6]}
       Location.Z {type=MovieSceneScriptingDoubleChannel, default=384.8292, const [384.8292,6]}
       Rotation.X {type=MovieSceneScriptingDoubleChannel, default=0, const [0,6]}
       Rotation.Y {type=MovieSceneScriptingDoubleChannel, default=0, const [0,6]}
       Rotation.Z {type=MovieSceneScriptingDoubleChannel, default=90, keys [[1952,90],[1958,142.2273],[1970,183.515],[2322,183.515],[2351,155.355],[2406,90]]}
       Scale.X {type=MovieSceneScriptingDoubleChannel, default=1, const [1,6]}
       Scale.Y {type=MovieSceneScriptingDoubleChannel, default=1, const [1,6]}
       Scale.Z {type=MovieSceneScriptingDoubleChannel, default=1, const [1,6]}
      ]
     }]
   }
  ]
 }
 B_BattleBus_Stamina {
  id=#28
  possessed_class=B_BattleBus_Stamina_C
  tracks [
   {type=MovieSceneSpawnTrack, sections [{type=MovieSceneSpawnSection, channels [None {type=MovieSceneScriptingBoolChannel, default=true}]}]}
   {
    type=MovieScene3DTransformTrack
    sections [{
      type=MovieScene3DTransformSection
      channels [
       Location.X {type=MovieSceneScriptingDoubleChannel, default=-19250.0933, keys [[3180,-19250.0938],[3279,-16774.5684]]}
       Location.Y {type=MovieSceneScriptingDoubleChannel, default=-25564.6301, keys [[3180,-25564.6309],[3279,-23150.627]]}
       Location.Z {type=MovieSceneScriptingDoubleChannel, default=38281.5445, const [38281.543,2]}
       Rotation.X {type=MovieSceneScriptingDoubleChannel, default=0, const [0,2]}
       Rotation.Y {type=MovieSceneScriptingDoubleChannel, default=0, const [0,2]}
       Rotation.Z {type=MovieSceneScriptingDoubleChannel, default=44.2791, const [44.2791,2]}
       Scale.X {type=MovieSceneScriptingDoubleChannel, default=1, const [1,2]}
       Scale.Y {type=MovieSceneScriptingDoubleChannel, default=1, const [1,2]}
       Scale.Z {type=MovieSceneScriptingDoubleChannel, default=1, const [1,2]}
      ]
     }]
   }
   {
    type=MovieSceneVisibilityTrack
    sections [{type=MovieSceneVisibilitySection, channels [None {type=MovieSceneScriptingBoolChannel, default=true, keys [[1001,false],[3128,true]]}]}]
   }
  ]
 }
 SM_WildEstate_House_PictureFrames_A {
  id=#29
  possessed_class=BP_WildEstate_House_PictureFrames_A_C
  tracks [{
    type=MovieSceneVisibilityTrack
    sections [{type=MovieSceneVisibilitySection, channels [None {type=MovieSceneScriptingBoolChannel, default=true, keys [[1001,false],[1722,true]]}]}]
   }]
 }
 _0_CameraRig_Rail {
  id=#30
  possessed_class=CameraRig_Rail
  tracks [
   {
    type=MovieSceneSpawnTrack
    sections [{type=MovieSceneSpawnSection, channels [None {type=MovieSceneScriptingBoolChannel, default=true, const [true,2]}]}]
   }
   {
    type=MovieScene3DTransformTrack
    sections [{
      type=MovieScene3DTransformSection
      channels [
       Location.X {type=MovieSceneScriptingDoubleChannel, default=7464.3386, thinned_from=8, keys [[1001,7464.3386],[5506,7464.3384]]}
       Location.Y {type=MovieSceneScriptingDoubleChannel, default=11012.2387, thinned_from=8, keys [[1001,11012.2387],[5506,11012.2383]]}
       Location.Z {type=MovieSceneScriptingDoubleChannel, default=422.674, thinned_from=8, keys [[1003,422.674],[5506,422.674]]}
       Rotation.X {type=MovieSceneScriptingDoubleChannel, default=0, const [0,8]}
       Rotation.Y {type=MovieSceneScriptingDoubleChannel, default=0, const [0,8]}
       Rotation.Z {type=MovieSceneScriptingDoubleChannel, default=86.8822, thinned_from=8, keys [[1003,86.8822],[5506,86.8822]]}
       Scale.X {type=MovieSceneScriptingDoubleChannel, default=1, const [1,8]}
       Scale.Y {type=MovieSceneScriptingDoubleChannel, default=1, const [1,8]}
       Scale.Z {type=MovieSceneScriptingDoubleChannel, default=1, const [1,8]}
      ]
     }]
   }
   {
    type=MovieSceneFloatTrack
    display_name=Current Position on Rail
    sections [{
      type=MovieSceneFloatSection
      channels [None {
        type=MovieSceneScriptingFloatChannel
        default=0.1026
        keys [[1001,0.0024],[1121,0.0024],[1175,0.0065],[1273,0.0175],[1532,0.0339],[1745,0.0417],[1815,0.0454],[1977,0.0634],[2419,0.1476],[2523,0.1537],[2623,0.1696],[2721,0.2095],[3186,0.4141],[3614,0.5006],[3967,0.5496],[4360,0.5969],[5304,0.7499],[6116,1]]
       }]
     }]
   }
  ]
 }
]
display_rate {denominator=1, numerator=60}
tick_resolution {denominator=1, numerator=24000}
tracks [
 {type=MovieSceneTimeWarpTrack, display_name=Time Warp Curve, sections [{type=MovieSceneTimeWarpSection}]}
 {type=MovieSceneCameraCutTrack, sections [{type=MovieSceneCameraCutSection, end_frame=13302, start_frame=737}]}
]
