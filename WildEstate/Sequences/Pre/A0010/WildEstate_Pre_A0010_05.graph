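# Level Sequence for WildEstate Pre shot A0010_05: one cine-camera move through the estate house, ending in a huge rising pullback before the Battle Bus appears.
# A spawned rail rig with attached orbit, pan/boom and tilt/roll control meshes layers the camera motion, alongside focal-length changes.
# Keyed visibility swaps walls, stairs and props during the move, while a pivot mesh swings a stand-in door and the TV and spotlight shift.
# A Time Warp track accompanies the single camera cut.
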LevelSequence /WildEstate/Sequences/Pre/A0010/WildEstate_Pre_A0010_05
bindings [
 BP_Blockout_Doorway_A_A_A75 {
  id=#1
  possessed_class=BP_WildEstate_Blockout_Doorway_Side_A_B_C
  tracks [{
    type=MovieSceneVisibilityTrack
    sections [{
      type=MovieSceneVisibilitySection
      channels [None {type=MovieSceneScriptingBoolChannel, default=false, keys [[1001,true],[1356,false],[1722,true]]}]
     }]
   }]
 }
 TestScaleEditorActor3 {
  id=#2
  possessed_class=TestScaleEditorActor_C
  tracks [{
    type=MovieSceneVisibilityTrack
    sections [{type=MovieSceneVisibilitySection, channels [None {type=MovieSceneScriptingBoolChannel, default=false}]}]
   }]
 }
 _A0010_WorldCube {
  id=#3
  possessed_class=FortStaticMeshActor
  tracks [
   {type=MovieSceneSpawnTrack, sections [{type=MovieSceneSpawnSection, channels [None {type=MovieSceneScriptingBoolChannel, default=true}]}]}
   {
    type=MovieScene3DTransformTrack
    sections [{
      type=MovieScene3DTransformSection
      channels [
       Location.X {type=MovieSceneScriptingDoubleChannel, default=7205.2965}
       Location.Y {type=MovieSceneScriptingDoubleChannel, default=11814.1699}
       Location.Z {type=MovieSceneScriptingDoubleChannel, default=366.018}
       Rotation.X {type=MovieSceneScriptingDoubleChannel, default=0}
       Rotation.Y {type=MovieSceneScriptingDoubleChannel, default=0}
       Rotation.Z {type=MovieSceneScriptingDoubleChannel, default=0}
       Scale.X {type=MovieSceneScriptingDoubleChannel, default=1}
       Scale.Y {type=MovieSceneScriptingDoubleChannel, default=1}
       Scale.Z {type=MovieSceneScriptingDoubleChannel, default=1}
      ]
     }]
   }
  ]
 }
 Cam_A0010 {
  id=#4
  possessed_class=CineCameraActor
  tracks [
   {
    type=MovieScene3DTransformTrack
    sections [{
      type=MovieScene3DTransformSection
      channels [
       Location.X {type=MovieSceneScriptingDoubleChannel, default=0, const [0,2]}
       Location.Y {type=MovieSceneScriptingDoubleChannel, default=0, const [0,2]}
       Location.Z {type=MovieSceneScriptingDoubleChannel, default=0, const [0,2]}
       Rotation.X {type=MovieSceneScriptingDoubleChannel, default=0, const [0,2]}
       Rotation.Y {type=MovieSceneScriptingDoubleChannel, default=0, const [0,2]}
       Rotation.Z {type=MovieSceneScriptingDoubleChannel, default=0, const [0,2]}
       Scale.X {type=MovieSceneScriptingDoubleChannel, default=1, const [1,2]}
       Scale.Y {type=MovieSceneScriptingDoubleChannel, default=1, const [1,2]}
       Scale.Z {type=MovieSceneScriptingDoubleChannel, default=1, const [1,2]}
      ]
     }]
   }
   {type=MovieSceneSpawnTrack, sections [{type=MovieSceneSpawnSection, channels [None {type=MovieSceneScriptingBoolChannel, default=true}]}]}
   {type=MovieScene3DAttachTrack, sections [{type=MovieScene3DAttachSection}]}
  ]
 }
 GeoScript_WildEstate_VTXColor2 {
  id=#5
  possessed_class=GeoScript_WildEstate_VTXColor_C
  tracks [{
    type=MovieSceneVisibilityTrack
    sections [{type=MovieSceneVisibilitySection, channels [None {type=MovieSceneScriptingBoolChannel, default=false}]}]
   }]
 }
 BP_Blockout_Stairs_Rail_A_TESTTT {
  id=#6
  possessed_class=BP_Estate_Stairs_Rail_A_B_C
  tracks [{
    type=MovieSceneVisibilityTrack
    sections [{
      type=MovieSceneVisibilitySection
      channels [None {type=MovieSceneScriptingBoolChannel, default=false, keys [[1001,true],[1376,false],[1722,true]]}]
     }]
   }]
 }
 BP_Blockout_Doorway_A_A_A9 {
  id=#7
  possessed_class=BP_WildEstate_Blockout_Wall_A_A_C
  tracks [{
    type=MovieSceneVisibilityTrack
    sections [{type=MovieSceneVisibilitySection, channels [None {type=MovieSceneScriptingBoolChannel, default=true, keys [[1001,false],[1722,true]]}]}]
   }]
 }
 SM_WildEstate_House_Door_A {
  id=#8
  possessed_class=FortStaticMeshActor
  tracks [
   {type=MovieSceneSpawnTrack, sections [{type=MovieSceneSpawnSection, channels [None {type=MovieSceneScriptingBoolChannel, default=true}]}]}
   {
    type=MovieScene3DTransformTrack
    sections [{
      type=MovieScene3DTransformSection
      channels [
       Location.X {type=MovieSceneScriptingDoubleChannel, default=-85.583}
       Location.Y {type=MovieSceneScriptingDoubleChannel, default=0}
       Location.Z {type=MovieSceneScriptingDoubleChannel, default=0}
       Rotation.X {type=MovieSceneScriptingDoubleChannel, default=0}
       Rotation.Y {type=MovieSceneScriptingDoubleChannel, default=0}
       Rotation.Z {type=MovieSceneScriptingDoubleChannel, default=0}
       Scale.X {type=MovieSceneScriptingDoubleChannel, default=1}
       Scale.Y {type=MovieSceneScriptingDoubleChannel, default=1}
       Scale.Z {type=MovieSceneScriptingDoubleChannel, default=1}
      ]
     }]
   }
   {type=MovieScene3DAttachTrack, sections [{type=MovieScene3DAttachSection}]}
  ]
 }
 BP_Blockout_Doorway_A_A_A73 {
  id=#9
  possessed_class=BP_WildEstate_Blockout_Doorway_Side_A_B_C
  tracks [{
    type=MovieSceneVisibilityTrack
    sections [{
      type=MovieSceneVisibilitySection
      channels [None {type=MovieSceneScriptingBoolChannel, default=false, keys [[1001,true],[1384,false],[1722,true]]}]
     }]
   }]
 }
 LightComponent0 {
  id=#10
  parent_id=#11
  possessed_class=SpotLightComponent
  tracks [
   {
    type=MovieSceneFloatTrack
    display_name=Intensity
    sections [{type=MovieSceneFloatSection, channels [None {type=MovieSceneScriptingFloatChannel, default=4}]}]
   }
   {
    type=MovieSceneColorTrack
    display_name=Light Color
    sections [{
      type=MovieSceneColorSection
      channels [
       Color.R {type=MovieSceneScriptingFloatChannel, default=1}
       Color.G {type=MovieSceneScriptingFloatChannel, default=1}
       Color.B {type=MovieSceneScriptingFloatChannel, default=1}
       Color.A {type=MovieSceneScriptingFloatChannel, default=1}
      ]
     }]
   }
  ]
 }
 BP_WildEstate_House_TV_A {
  id=#12
  possessed_class=BP_WildEstate_House_TV_A_C
  tracks [{
    type=MovieScene3DTransformTrack
    sections [{
      type=MovieScene3DTransformSection
      channels [
       Location.X {type=MovieSceneScriptingDoubleChannel, default=7265.8323, keys [[1001,7220.3354],[1439,7265.832]]}
       Location.Y {type=MovieSceneScriptingDoubleChannel, default=11572.4742, keys [[1001,11541.8691],[1439,11560.4746]]}
       Location.Z {type=MovieSceneScriptingDoubleChannel, default=394.9222, keys [[1001,378.3755],[1439,394.9221]]}
       Rotation.X {type=MovieSceneScriptingDoubleChannel, default=0, const [0,2]}
       Rotation.Y {type=MovieSceneScriptingDoubleChannel, default=0, const [0,2]}
       Rotation.Z {type=MovieSceneScriptingDoubleChannel, default=0, const [0,2]}
       Scale.X {type=MovieSceneScriptingDoubleChannel, default=1, const [1,2]}
       Scale.Y {type=MovieSceneScriptingDoubleChannel, default=1, const [1,2]}
       Scale.Z {type=MovieSceneScriptingDoubleChannel, default=1, const [1,2]}
      ]
     }]
   }]
 }
 TestScaleEditorActor2 {
  id=#13
  possessed_class=TestScaleEditorActor_C
  tracks [{
    type=MovieSceneVisibilityTrack
    sections [{type=MovieSceneVisibilitySection, channels [None {type=MovieSceneScriptingBoolChannel, default=false}]}]
   }]
 }
 DoorComponent {
  id=#14
  parent_id=#15
  possessed_class=FortBuildingWallDoorComponent
  tracks [
   {
    type=MovieScene3DTransformTrack
    sections [{
      type=MovieScene3DTransformSection
      channels [
       Location.X {type=MovieSceneScriptingDoubleChannel, default=0}
       Location.Y {type=MovieSceneScriptingDoubleChannel, default=-1}
       Location.Z {type=MovieSceneScriptingDoubleChannel, default=1}
       Rotation.X {type=MovieSceneScriptingDoubleChannel, default=0}
       Rotation.Y {type=MovieSceneScriptingDoubleChannel, default=0}
       Rotation.Z {type=MovieSceneScriptingDoubleChannel, default=0}
       Scale.X {type=MovieSceneScriptingDoubleChannel, default=1}
       Scale.Y {type=MovieSceneScriptingDoubleChannel, default=1}
       Scale.Z {type=MovieSceneScriptingDoubleChannel, default=1}
      ]
     }]
   }
   {
    type=MovieSceneVisibilityTrack
    sections [{type=MovieSceneVisibilitySection, channels [None {type=MovieSceneScriptingBoolChannel, default=false}]}]
   }
  ]
 }
 BrickSimple_Chimney_B {
  id=#16
  possessed_class=BP_WildEstate_House_Fireplace_A_C
  tracks [{
    type=MovieSceneVisibilityTrack
    sections [{type=MovieSceneVisibilitySection, channels [None {type=MovieSceneScriptingBoolChannel, default=true, keys [[1001,false],[1384,true]]}]}]
   }]
 }
 HoodLOD0_LOD0 {
  id=#17
  possessed_class=FortStaticMeshActor
  tracks [
   {
    type=MovieScene3DTransformTrack
    sections [{
      type=MovieScene3DTransformSection
      channels [
       Location.X {type=MovieSceneScriptingDoubleChannel, default=16952.6667}
       Location.Y {type=MovieSceneScriptingDoubleChannel, default=3160.7244}
       Location.Z {type=MovieSceneScriptingDoubleChannel, default=10.8123}
       Rotation.X {type=MovieSceneScriptingDoubleChannel, default=0}
       Rotation.Y {type=MovieSceneScriptingDoubleChannel, default=0}
       Rotation.Z {type=MovieSceneScriptingDoubleChannel, default=157.4188}
       Scale.X {type=MovieSceneScriptingDoubleChannel, default=1}
       Scale.Y {type=MovieSceneScriptingDoubleChannel, default=1}
       Scale.Z {type=MovieSceneScriptingDoubleChannel, default=1}
      ]
     }]
   }
   {
    type=MovieSceneVisibilityTrack
    sections [{type=MovieSceneVisibilitySection, channels [None {type=MovieSceneScriptingBoolChannel, default=false}]}]
   }
  ]
 }
 _1_BaseOrbit_CTRL {
  id=#18
  possessed_class=FortStaticMeshActor
  tracks [
   {type=MovieSceneSpawnTrack, sections [{type=MovieSceneSpawnSection, channels [None {type=MovieSceneScriptingBoolChannel, default=true}]}]}
   {
    type=MovieScene3DTransformTrack
    sections [{
      type=MovieScene3DTransformSection
      channels [
       Location.X {type=MovieSceneScriptingDoubleChannel, default=18.5133, const [0,16]}
       Location.Y {type=MovieSceneScriptingDoubleChannel, default=0, const [0,16]}
       Location.Z {type=MovieSceneScriptingDoubleChannel, default=0, const [0,16]}
       Rotation.X {type=MovieSceneScriptingDoubleChannel, default=0, thinned_from=13, keys [[1001,0],[2994,0]]}
       Rotation.Y {type=MovieSceneScriptingDoubleChannel, default=0, thinned_from=13, keys [[1001,0],[2994,0]]}
       Rotation.Z {
        type=MovieSceneScriptingDoubleChannel
        default=114.737
        keys [[1001,28.6135],[1121,28.6135],[1140,27.5925],[1261,8.2298],[1341,-15.1319],[1350.0025,-19.0056],[1418,-52.5661],[1462,-68.7569],[1544,-96.1077],[1639,-119.4449],[1759,-132.4445],[1850,-161.5789],[1914,-266.5968],[1949,-274.96],[2023,-274.96],[2417,-409.9868],[2523,-413.4572]]
       }
       Scale.X {type=MovieSceneScriptingDoubleChannel, default=1, const [1,16]}
       Scale.Y {type=MovieSceneScriptingDoubleChannel, default=1, const [1,16]}
       Scale.Z {type=MovieSceneScriptingDoubleChannel, default=1, const [1,16]}
      ]
     }]
   }
   {type=MovieScene3DAttachTrack, sections [{type=MovieScene3DAttachSection}]}
  ]
 }
 BP_Blockout_Doorway_A_A_A8 {
  id=#19
  possessed_class=BP_WildEstate_House_Window_C_A_C
  tracks [{
    type=MovieSceneVisibilityTrack
    sections [{type=MovieSceneVisibilitySection, channels [None {type=MovieSceneScriptingBoolChannel, default=true, keys [[1001,false],[1722,true]]}]}]
   }]
 }
 SpotLight {
  id=#11
  possessed_class=SpotLight
  tracks [{
    type=MovieScene3DTransformTrack
    sections [{
      type=MovieScene3DTransformSection
      channels [
       Location.X {type=MovieSceneScriptingDoubleChannel, default=7264, keys [[1001,7240.2725],[1439,7264]]}
       Location.Y {type=MovieSceneScriptingDoubleChannel, default=11552, keys [[1001,11608.0605],[1439,11572]]}
       Location.Z {type=MovieSceneScriptingDoubleChannel, default=466, keys [[1001,440.0335],[1439,466]]}
       Rotation.X {type=MovieSceneScriptingDoubleChannel, default=-90, const [-90,2]}
       Rotation.Y {type=MovieSceneScriptingDoubleChannel, default=-2.8125, const [-2.8125,2]}
       Rotation.Z {type=MovieSceneScriptingDoubleChannel, default=90, const [90,2]}
       Scale.X {type=MovieSceneScriptingDoubleChannel, default=1, const [1,2]}
       Scale.Y {type=MovieSceneScriptingDoubleChannel, default=1, const [1,2]}
       Scale.Z {type=MovieSceneScriptingDoubleChannel, default=1, const [1,2]}
      ]
     }]
   }]
 }
 _3_TiltRoll_CTRL {
  id=#20
  possessed_class=FortStaticMeshActor
  tracks [
   {type=MovieSceneSpawnTrack, sections [{type=MovieSceneSpawnSection, channels [None {type=MovieSceneScriptingBoolChannel, default=true}]}]}
   {
    type=MovieScene3DTransformTrack
    sections [{
      type=MovieScene3DTransformSection
      channels [
       Location.X {type=MovieSceneScriptingDoubleChannel, default=0, thinned_from=10, keys [[1001,-55.3521],[1121,-0.8799],[2994,-0.8799]]}
       Location.Y {type=MovieSceneScriptingDoubleChannel, default=0, thinned_from=10, keys [[1001,0],[2994,0]]}
       Location.Z {type=MovieSceneScriptingDoubleChannel, default=0, thinned_from=10, keys [[1001,9.4262],[1121,0.1498],[2994,0.1498]]}
       Rotation.X {type=MovieSceneScriptingDoubleChannel, default=0, thinned_from=10, keys [[1001,0],[2994,0]]}
       Rotation.Y {
        type=MovieSceneScriptingDoubleChannel
        default=-7.6791
        keys [[1001,-9.6645],[1121,-9.6645],[1261,-5.2708],[1350,-2.8605],[1924,-2.8605],[2172,-1.2971],[2416,1.0649],[2523,1.507],[2599,-18.4659],[2677,-25.2092],[2994,-30.3541]]
       }
       Rotation.Z {type=MovieSceneScriptingDoubleChannel, default=0, thinned_from=10, keys [[1001,0],[2994,0]]}
       Scale.X {type=MovieSceneScriptingDoubleChannel, default=1, const [1,10]}
       Scale.Y {type=MovieSceneScriptingDoubleChannel, default=1, const [1,10]}
       Scale.Z {type=MovieSceneScriptingDoubleChannel, default=1, const [1,10]}
      ]
     }]
   }
   {type=MovieScene3DAttachTrack, sections [{type=MovieScene3DAttachSection}]}
  ]
 }
 CameraComponent {
  id=#21
  parent_id=#4
  possessed_class=CineCameraComponent
  tracks [
   {
    type=MovieSceneFloatTrack
    display_name=Current Focal Length
    sections [{
      type=MovieSceneFloatSection
      channels [None {type=MovieSceneScriptingFloatChannel, default=24, keys [[1121,30],[1274,17],[2075,17],[2253,24]]}]
     }]
   }
   {
    type=MovieSceneFloatTrack
    display_name="Manual Focus Distance (Focus Settings)"
    sections [{type=MovieSceneFloatSection, channels [None {type=MovieSceneScriptingFloatChannel, default=0}]}]
   }
   {
    type=MovieSceneFloatTrack
    display_name=Current Aperture
    sections [{type=MovieSceneFloatSection, channels [None {type=MovieSceneScriptingFloatChannel, default=2.8}]}]
   }
   {
    type=MovieScene3DTransformTrack
    sections [{
      type=MovieScene3DTransformSection
      channels [
       Location.X {type=MovieSceneScriptingDoubleChannel, default=0, const [0,4]}
       Location.Y {type=MovieSceneScriptingDoubleChannel, default=0, const [0,4]}
       Location.Z {type=MovieSceneScriptingDoubleChannel, default=0, const [0,4]}
       Rotation.X {type=MovieSceneScriptingDoubleChannel, default=0, const [0,4]}
       Rotation.Y {type=MovieSceneScriptingDoubleChannel, default=0, const [0,4]}
       Rotation.Z {type=MovieSceneScriptingDoubleChannel, default=0, const [0,4]}
       Scale.X {type=MovieSceneScriptingDoubleChannel, default=1, const [1,4]}
       Scale.Y {type=MovieSceneScriptingDoubleChannel, default=1, const [1,4]}
       Scale.Z {type=MovieSceneScriptingDoubleChannel, default=1, const [1,4]}
      ]
     }]
   }
  ]
 }
 _2_PanBoom_CTRL {
  id=#22
  possessed_class=FortStaticMeshActor
  tracks [
   {type=MovieSceneSpawnTrack, sections [{type=MovieSceneSpawnSection, channels [None {type=MovieSceneScriptingBoolChannel, default=true}]}]}
   {
    type=MovieScene3DTransformTrack
    sections [{
      type=MovieScene3DTransformSection
      channels [
       Location.X {
        type=MovieSceneScriptingDoubleChannel
        default=-806.4399
        keys [[1121,94.905],[1261,30.5877],[1350,0],[1554,-148.9901],[1697,-200.4992],[1924,-200.4992],[2172,-276.2491],[2416,-455.7071],[2523,-493.9005],[2994,-68992.3047]]
       }
       Location.Y {
        type=MovieSceneScriptingDoubleChannel
        default=-7.2342
        thinned_from=10
        keys [[1121,-123.1737],[1261,-39.6986],[1350,0],[1845,0],[1924,31.3263],[1984,0],[2523,0],[2994,-691.6379]]
       }
       Location.Z {
        type=MovieSceneScriptingDoubleChannel
        default=252.6591
        thinned_from=10
        keys [[1121,204.4086],[1261,150.6534],[1350,111.0947],[1924,111.0947],[2172,185.4804],[2418,601.8976],[2523,610.6801],[2994,46168.8672]]
       }
       Rotation.X {type=MovieSceneScriptingDoubleChannel, default=0, const [0,9]}
       Rotation.Y {type=MovieSceneScriptingDoubleChannel, default=0, const [0,9]}
       Rotation.Z {type=MovieSceneScriptingDoubleChannel, default=-0.582, const [0.5785,9]}
       Scale.X {type=MovieSceneScriptingDoubleChannel, default=1, const [1,9]}
       Scale.Y {type=MovieSceneScriptingDoubleChannel, default=1, const [1,9]}
       Scale.Z {type=MovieSceneScriptingDoubleChannel, default=1, const [1,9]}
      ]
     }]
   }
   {type=MovieScene3DAttachTrack, sections [{type=MovieScene3DAttachSection}]}
  ]
 }
 BP_Blockout_Doorway_A_A_A17 {
  id=#23
  possessed_class=BP_WildEstate_Blockout_Wall_A_A_C
  tracks [{
    type=MovieSceneVisibilityTrack
    sections [{
      type=MovieSceneVisibilitySection
      channels [None {type=MovieSceneScriptingBoolChannel, default=false, keys [[1001,true],[1384,false],[1722,true]]}]
     }]
   }]
 }
 BP_Blockout_Doorway_A_A_A {id=#15, possessed_class=BP_WildEstate_House_Door_A_A_C, tracks []}
 Artemis_Suburban_CurtainsModern_C3 {
  id=#24
  possessed_class=Artemis_Suburban_CurtainsModern_C_C
  tracks [{
    type=MovieSceneVisibilityTrack
    sections [{type=MovieSceneVisibilitySection, channels [None {type=MovieSceneScriptingBoolChannel, default=true, keys [[1001,false],[1722,true]]}]}]
   }]
 }
 door_pivot {
  id=#25
  possessed_class=FortStaticMeshActor
  tracks [
   {type=MovieSceneSpawnTrack, sections [{type=MovieSceneSpawnSection, channels [None {type=MovieSceneScriptingBoolChannel, default=true}]}]}
   {
    type=MovieScene3DTransformTrack
    sections [{
      type=MovieScene3DTransformSection
      channels [
       Location.X {type=MovieSceneScriptingDoubleChannel, default=6400.9955, const [6400.9956,6]}
       Location.Y {type=MovieSceneScriptingDoubleChannel, default=12288, const [12373.583,6]}
       Location.Z {type=MovieSceneScriptingDoubleChannel, default=384.8292, const [384.8292,6]}
       Rotation.X {type=MovieSceneScriptingDoubleChannel, default=0, const [0,6]}
       Rotation.Y {type=MovieSceneScriptingDoubleChannel, default=0, const [0,6]}
       Rotation.Z {type=MovieSceneScriptingDoubleChannel, default=90, keys [[1952,90],[1958,142.2273],[1970,183.515],[2322,183.515],[2351,155.355],[2406,90]]}
       Scale.X {type=MovieSceneScriptingDoubleChannel, default=1, const [1,6]}
       Scale.Y {type=MovieSceneScriptingDoubleChannel, default=1, const [1,6]}
       Scale.Z {type=MovieSceneScriptingDoubleChannel, default=1, const [1,6]}
      ]
     }]
   }
  ]
 }
 B_BattleBus_Stamina {
  id=#26
  possessed_class=B_BattleBus_Stamina_C
  tracks [
   {type=MovieSceneSpawnTrack, sections [{type=MovieSceneSpawnSection, channels [None {type=MovieSceneScriptingBoolChannel, default=true}]}]}
   {
    type=MovieScene3DTransformTrack
    sections [{
      type=MovieScene3DTransformSection
      channels [
       Location.X {type=MovieSceneScriptingDoubleChannel, default=-19250.0933, keys [[3180,-19250.0938],[3279,-16774.5684]]}
       Location.Y {type=MovieSceneScriptingDoubleChannel, default=-25564.6301, keys [[3180,-25564.6309],[3279,-23150.627]]}
       Location.Z {type=MovieSceneScriptingDoubleChannel, default=38281.5445, const [38281.543,2]}
       Rotation.X {type=MovieSceneScriptingDoubleChannel, default=0, const [0,2]}
       Rotation.Y {type=MovieSceneScriptingDoubleChannel, default=0, const [0,2]}
       Rotation.Z {type=MovieSceneScriptingDoubleChannel, default=44.2791, const [44.2791,2]}
       Scale.X {type=MovieSceneScriptingDoubleChannel, default=1, const [1,2]}
       Scale.Y {type=MovieSceneScriptingDoubleChannel, default=1, const [1,2]}
       Scale.Z {type=MovieSceneScriptingDoubleChannel, default=1, const [1,2]}
      ]
     }]
   }
   {
    type=MovieSceneVisibilityTrack
    sections [{type=MovieSceneVisibilitySection, channels [None {type=MovieSceneScriptingBoolChannel, default=true, keys [[1001,false],[3128,true]]}]}]
   }
  ]
 }
 SM_WildEstate_House_PictureFrames_A {
  id=#27
  possessed_class=BP_WildEstate_House_PictureFrames_A_C
  tracks [{
    type=MovieSceneVisibilityTrack
    sections [{type=MovieSceneVisibilitySection, channels [None {type=MovieSceneScriptingBoolChannel, default=true, keys [[1001,false],[1722,true]]}]}]
   }]
 }
 _0_CameraRig_Rail {
  id=#28
  possessed_class=CameraRig_Rail
  tracks [
   {
    type=MovieSceneSpawnTrack
    sections [{type=MovieSceneSpawnSection, channels [None {type=MovieSceneScriptingBoolChannel, default=true, const [true,2]}]}]
   }
   {
    type=MovieScene3DTransformTrack
    sections [{
      type=MovieScene3DTransformSection
      channels [
       Location.X {type=MovieSceneScriptingDoubleChannel, default=7464.3386, thinned_from=8, keys [[1001,7464.3386],[5506,7464.3384]]}
       Location.Y {type=MovieSceneScriptingDoubleChannel, default=11012.2387, thinned_from=8, keys [[1001,11012.2387],[5506,11012.2383]]}
       Location.Z {type=MovieSceneScriptingDoubleChannel, default=422.674, thinned_from=8, keys [[1003,422.674],[5506,422.674]]}
       Rotation.X {type=MovieSceneScriptingDoubleChannel, default=0, const [0,8]}
       Rotation.Y {type=MovieSceneScriptingDoubleChannel, default=0, const [0,8]}
       Rotation.Z {type=MovieSceneScriptingDoubleChannel, default=86.8822, thinned_from=8, keys [[1003,86.8822],[5506,86.8822]]}
       Scale.X {type=MovieSceneScriptingDoubleChannel, default=1, const [1,8]}
       Scale.Y {type=MovieSceneScriptingDoubleChannel, default=1, const [1,8]}
       Scale.Z {type=MovieSceneScriptingDoubleChannel, default=1, const [1,8]}
      ]
     }]
   }
   {
    type=MovieSceneFloatTrack
    display_name=Current Position on Rail
    sections [{
      type=MovieSceneFloatSection
      channels [None {
        type=MovieSceneScriptingFloatChannel
        default=0.1026
        keys [[1001,0.0024],[1121,0.0024],[1175,0.0065],[1273,0.0175],[1532,0.0339],[1745,0.0417],[1815,0.0454],[1977,0.0634],[2419,0.1476],[2523,0.1537]]
       }]
     }]
   }
  ]
 }
]
display_rate {denominator=1, numerator=60}
tick_resolution {denominator=1, numerator=24000}
tracks [
 {type=MovieSceneTimeWarpTrack, display_name=Time Warp Curve, sections [{type=MovieSceneTimeWarpSection}]}
 {type=MovieSceneCameraCutTrack, sections [{type=MovieSceneCameraCutSection, end_frame=13302, start_frame=737}]}
]
